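# Physics asset for the F_MED_SweetLetter female body cosmetic, setting up secondary motion for its chest fur and chest O-ring.
# Small capsules on base bones, forearms and root link to the pelvis capsule through constraints free in swing and twist.
# Each dynamic fur capsule hangs off its base bone with tight swing and twist limits.
# The O-ring follows the same base-to-dynamic pattern but allows a much wider first swing.
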PhysicsAsset /BRCosmetics/Characters/Player/Female/Medium/Bodies/F_MED_SweetLetter/Meshes/F_MED_SweetLetter_Physics
bodies {
 base_ChestFur_01_l [{shapeType=Capsule, length=0.505, radius=0.505, shapeName=base_ChestFur_01_l_capsule, rotation=<rot 0 -90 0>}]
 base_ChestFur_01_r [{shapeType=Capsule, length=0.505, radius=0.505, shapeName=base_ChestFur_01_r_capsule, rotation=<rot 0 -90 0>}]
 base_ChestFur_02_l [{shapeType=Capsule, length=0.505, radius=0.505, shapeName=base_ChestFur_02_l_capsule, rotation=<rot 0 -90 0>}]
 base_ChestFur_02_r [{shapeType=Capsule, length=0.505, radius=0.505, shapeName=base_ChestFur_02_r_capsule, rotation=<rot 0 -90 0>}]
 base_ChestFur_03_l [{shapeType=Capsule, length=0.505, radius=0.505, shapeName=base_ChestFur_03_l_capsule, rotation=<rot 0 -90 0>}]
 base_ChestFur_03_r [{shapeType=Capsule, length=0.505, radius=0.505, shapeName=base_ChestFur_03_r_capsule, rotation=<rot 0 -90 0>}]
 base_chestORing [{
   shapeType=Capsule
   length=0.505
   radius=0.505
   shapeName=base_chestORing_capsule
   center=<0.0343 0.0591 0.0644>
   rotation=<rot 0 0.0764 -87.049>
  }]
 dyn_ChestFur_01_l [{shapeType=Capsule, length=5.452, radius=1.7418, shapeName=dyn_ChestFur_01_l_capsule, center=<3.349 0 0.9841>, rotation=<rot 90 0 0>}]
 dyn_ChestFur_01_r [{shapeType=Capsule, length=5.452, radius=1.7418, shapeName=dyn_ChestFur_01_l_capsule, center=<3.349 0 -0.9841>, rotation=<rot 90 0 0>}]
 dyn_ChestFur_02_l [{shapeType=Capsule, length=5.452, radius=1.7418, shapeName=dyn_ChestFur_02_l_capsule, center=<3.349 0 0.9841>, rotation=<rot 90 0 0>}]
 dyn_ChestFur_02_r [{shapeType=Capsule, length=5.452, radius=1.7418, shapeName=dyn_ChestFur_02_l_capsule, center=<3.349 0 -0.9841>, rotation=<rot 90 0 0>}]
 dyn_ChestFur_03_l [{shapeType=Capsule, length=5.452, radius=1.7418, shapeName=dyn_ChestFur_03_l_capsule, center=<3.349 0 0.9841>, rotation=<rot 90 0 0>}]
 dyn_ChestFur_03_r [{shapeType=Capsule, length=5.452, radius=1.7418, shapeName=dyn_ChestFur_03_l_capsule, center=<3.349 0 -0.9841>, rotation=<rot 90 0 0>}]
 dyn_chestORing [{shapeType=Capsule, length=1.759, radius=1.7286, shapeName=dyn_chestORing_capsule, center=<2 0 0>, rotation=<rot 90 0 0>}]
 lowerarm_l [{shapeType=Capsule, length=0.505, radius=0.505, shapeName=lowerarm_l_capsule, rotation=<rot 0 -90 0>}]
 lowerarm_r [{shapeType=Capsule, length=0.505, radius=0.505, shapeName=lowerarm_r_capsule, rotation=<rot 0 -90 0>}]
 pelvis [{
   shapeType=Capsule
   length=15.0105
   radius=12.6167
   shapeName=pelvis_capsule
   center=<0.1893 0.5453 -0.0017>
   rotation=<rot 0.1519 -90 -0.1683>
  }]
 root [{shapeType=Capsule, length=0.505, radius=0.505, shapeName=root_capsule, rotation=<rot 0 -90 0>}]
}
constraints [
 {
  bone1Name=pelvis
  bone2Name=root
  swing1LimitDegrees=45
  swing1Motion=Free
  swing2LimitDegrees=45
  swing2Motion=Free
  twistLimitDegrees=45
  twistMotion=Free
 }
 {
  bone1Name=base_ChestFur_01_l
  bone2Name=pelvis
  swing1LimitDegrees=45
  swing1Motion=Free
  swing2LimitDegrees=45
  swing2Motion=Free
  twistLimitDegrees=45
  twistMotion=Free
 }
 {bone1Name=dyn_ChestFur_01_l, bone2Name=base_ChestFur_01_l, swing1LimitDegrees=2, swing2LimitDegrees=10, twistLimitDegrees=1}
 {
  bone1Name=base_ChestFur_01_r
  bone2Name=pelvis
  swing1LimitDegrees=45
  swing1Motion=Free
  swing2LimitDegrees=45
  swing2Motion=Free
  twistLimitDegrees=45
  twistMotion=Free
 }
 {bone1Name=dyn_ChestFur_01_r, bone2Name=base_ChestFur_01_r, swing1LimitDegrees=2, swing2LimitDegrees=10, twistLimitDegrees=1}
 {
  bone1Name=base_ChestFur_02_l
  bone2Name=pelvis
  swing1LimitDegrees=45
  swing1Motion=Free
  swing2LimitDegrees=45
  swing2Motion=Free
  twistLimitDegrees=45
  twistMotion=Free
 }
 {bone1Name=dyn_ChestFur_02_l, bone2Name=base_ChestFur_02_l, swing1LimitDegrees=2, swing2LimitDegrees=10, twistLimitDegrees=1}
 {
  bone1Name=base_ChestFur_02_r
  bone2Name=pelvis
  swing1LimitDegrees=45
  swing1Motion=Free
  swing2LimitDegrees=45
  swing2Motion=Free
  twistLimitDegrees=45
  twistMotion=Free
 }
 {bone1Name=dyn_ChestFur_02_r, bone2Name=base_ChestFur_02_r, swing1LimitDegrees=2, swing2LimitDegrees=5, twistLimitDegrees=1}
 {
  bone1Name=base_ChestFur_03_l
  bone2Name=pelvis
  swing1LimitDegrees=45
  swing1Motion=Free
  swing2LimitDegrees=45
  swing2Motion=Free
  twistLimitDegrees=45
  twistMotion=Free
 }
 {bone1Name=dyn_ChestFur_03_l, bone2Name=base_ChestFur_03_l, swing1LimitDegrees=2, swing2LimitDegrees=10, twistLimitDegrees=1}
 {
  bone1Name=base_ChestFur_03_r
  bone2Name=pelvis
  swing1LimitDegrees=45
  swing1Motion=Free
  swing2LimitDegrees=45
  swing2Motion=Free
  twistLimitDegrees=45
  twistMotion=Free
 }
 {bone1Name=dyn_ChestFur_03_r, bone2Name=base_ChestFur_03_r, swing1LimitDegrees=2, swing2LimitDegrees=10, twistLimitDegrees=1}
 {
  bone1Name=lowerarm_l
  bone2Name=pelvis
  swing1LimitDegrees=45
  swing1Motion=Free
  swing2LimitDegrees=45
  swing2Motion=Free
  twistLimitDegrees=45
  twistMotion=Free
 }
 {
  bone1Name=lowerarm_r
  bone2Name=pelvis
  swing1LimitDegrees=45
  swing1Motion=Free
  swing2LimitDegrees=45
  swing2Motion=Free
  twistLimitDegrees=45
  twistMotion=Free
 }
 {
  bone1Name=base_chestORing
  bone2Name=pelvis
  swing1LimitDegrees=45
  swing1Motion=Free
  swing2LimitDegrees=45
  swing2Motion=Free
  twistLimitDegrees=45
  twistMotion=Free
 }
 {bone1Name=dyn_chestORing, bone2Name=base_chestORing, swing1LimitDegrees=35, swing2LimitDegrees=10, twistLimitDegrees=1}
]
