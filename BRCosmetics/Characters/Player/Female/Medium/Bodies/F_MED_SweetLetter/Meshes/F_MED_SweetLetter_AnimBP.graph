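# Post-animation Blueprint for the SweetLetter female medium cosmetic body, copying the attached parent mesh's pose and adding secondary motion.
# A Control Rig fed an emote flag leads into weighted constraints for O-ring, buckle and zippers, then RigidBody and per-bone AnimDynamics for knee, elbow and wrist fur, linked by cached poses.
# Final additive bone offsets adjust the O-ring, plus left chest fur during grind rails.
# Updates refresh RigidBody gravity through a deprecated helper.
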AnimBlueprint /BRCosmetics/Characters/Player/Female/Medium/Bodies/F_MED_SweetLetter/Meshes/F_MED_SweetLetter_AnimBP
target_skeleton=&/F_MED_SweetLetter_Skeleton
graphs [
 AnimGraph {nodes [
   n1 = "Transform (Modify) Bone - Bone: base_ChestFur_01_l"(
    ComponentPose=Local To Component(LocalPose=Use cached pose 'Dynamics'())
    Alpha=Get Grind Rail() -> (Grind Rail:double)
    Translation="0,0,0"
    Rotation="15,0,-10"
    Scale="1,1,1"
   ) {
    rotationMode=BMM_Additive
    rotationSpace=BCS_BoneSpace
    alphaBoolBlend {blendOption=Linear}
    alphaScaleBiasClamp {bInterpResult=true}
    boneToModify {boneName=base_ChestFur_01_l}
    rotation {pitch=15, roll=-10}
   }
   n2 = "Transform (Modify) Bone - Bone: base_chestORing"(ComponentPose=@n1, Alpha=1, Translation="0,0,0", Rotation="0,-8,0", Scale="1,1,1") {
    rotationMode=BMM_Additive
    rotationSpace=BCS_BoneSpace
    alphaBoolBlend {blendOption=Linear}
    alphaScaleBiasClamp {bInterpResult=true}
    boneToModify {boneName=base_chestORing}
    rotation {yaw=-8}
   }
   Output Pose(Result=Component To Local(ComponentPose=@n2)) {layerGroup=None}
   n3 = Control Rig!(
    Source=Copy Pose From Mesh() {bUseAttachedParent=true}
    Alpha
    Playing Emote:double=Get Is Playing Emote CR() -> (Is Playing Emote CR:double)
   ) -> (Pose) {
    alphaBoolBlend {blendOption=Linear}
    controlRigAssetReference {blueprintRigClass=&/F_MED_SweetLetter_CtrlRig.F_MED_SweetLetter_CtrlRig_C}
   }
   Save cached pose 'CR Start'(Pose=@n3) {cacheName=CR Start}
   n4 = "Constraint - base_chestORing (PPP)"!(
    ComponentPose=Local To Component(LocalPose=Use cached pose 'CR Start'())
    Alpha=1
    ConstraintWeights_0=0.6873
    ConstraintWeights_1:float=0.2463
    ConstraintWeights_2:float=0.0663
   ) -> (Pose) {
    alphaBoolBlend {blendOption=Linear}
    boneToModify {boneName=base_chestORing}
    constraintSetup [{targetBone {boneName=spine_04}}, {targetBone {boneName=spine_05}}, {targetBone {boneName=deform_pec_r}}]
    constraintWeights [1,1,1]
   }
   n5 = "Constraint - base_beltBuckle (PPPPPPPP)"!(
    ComponentPose=@n4
    Alpha=1
    ConstraintWeights_0=0.8565
    ConstraintWeights_1:float=0.0457
    ConstraintWeights_2:float=0.0457
    ConstraintWeights_3:float=0.0176
    ConstraintWeights_4:float=0.0176
    ConstraintWeights_5:float=0.0159
    ConstraintWeights_6:float=0.0008
    ConstraintWeights_7:float=0.0002
   ) -> (Pose) {
    alphaBoolBlend {blendOption=Linear}
    boneToModify {boneName=base_beltBuckle}
    constraintSetup [
     {targetBone {boneName=pelvis}}
     {targetBone {boneName=deform_groin_r}}
     {targetBone {boneName=deform_groin_l}}
     {targetBone {boneName=thigh_twist_01_l}}
     {targetBone {boneName=thigh_twist_01_r}}
     {targetBone {boneName=spine_02}}
     {targetBone {boneName=spine_04}}
     {targetBone {boneName=spine_03}}
    ]
    constraintWeights [1,1,1,1,1,1,1,1]
   }
   n6 = "Constraint - base_forearmZipper_r (PPP)"!(ComponentPose=@n5, Alpha=1, ConstraintWeights_0=0.5301, ConstraintWeights_1:float=0.2513, ConstraintWeights_2:float=0.2186) -> (Pose) {
    alphaBoolBlend {blendOption=Linear}
    boneToModify {boneName=base_forearmZipper_r}
    constraintSetup [{targetBone {boneName=lowerarm_twist_01_r}}, {targetBone {boneName=lowerarm_twist_02_r}}, {targetBone {boneName=lowerarm_r}}]
    constraintWeights [1,1,1]
   }
   n7 = "Constraint - base_forearmZipper_l (PPP)"!(ComponentPose=@n6, Alpha=1, ConstraintWeights_0=0.5301, ConstraintWeights_1:float=0.2513, ConstraintWeights_2:float=0.2186) -> (Pose) {
    alphaBoolBlend {blendOption=Linear}
    boneToModify {boneName=base_forearmZipper_l}
    constraintSetup [{targetBone {boneName=lowerarm_twist_01_l}}, {targetBone {boneName=lowerarm_twist_02_l}}, {targetBone {boneName=lowerarm_l}}]
    constraintWeights [1,1,1]
   }
   Save cached pose 'Constraints'(Pose=Component To Local(ComponentPose=@n7)) {cacheName=Constraints}
   n8 = RigidBody!(
    ComponentPose=Local To Component(LocalPose=Use cached pose 'Constraints'())
    Alpha
    OverrideWorldGravity:Vector=Get GravityOverride()
    ExternalForce
   ) -> (Pose) {
    bOverrideWorldGravity=true
    overridePhysicsAsset=&/F_MED_SweetLetter_Physics
    simulationSpace=BaseBoneSpace
    alphaBoolBlend {blendOption=Linear}
    baseBoneRef {boneName=spine_03}
    componentAppliedLinearAccClamp=<2000 2000 2000>
    componentLinearAccScale=<0.25 0.15 0.15>
    componentLinearVelScale=<0.25 0.15 0.15>
    simSpaceSettings {
     maxAngularAcceleration=1000
     maxAngularVelocity=3
     maxLinearAcceleration=700
     maxLinearVelocity=500
     velocityScaleZ=0
     worldAlpha=0.2
     externalLinearDragV=<0.2 0.2>
    }
   }
   n9 = Anim Dynamics - Bone: dyn_KneeFur_02_l(ComponentPose=@n8, Alpha=1, ExternalForce="0,0,0") {
    angularDampingOverride=0.998
    angularSpringConstant=10000
    bAngularSpring=true
    bOverrideAngularDamping=true
    bUseGravityOverride=true
    alphaBoolBlend {blendOption=Linear}
    boundBone {boneName=dyn_KneeFur_02_l}
    componentAppliedLinearAccClamp=<3000 3000 3000>
    componentLinearAccScale=<0.4 0.4 0.4>
    componentLinearVelScale=<0.3 0.3 0.3>
    gravityOverride {z=-980}
    physicsBodyDefinitions [{
      boundBone {boneName=dyn_KneeFur_02_l}
      boxExtents=<25 25 25>
      constraintSetup {angularLimitsMax=<10 15 20>, angularLimitsMin=<-10 -10 -20>, angularTarget=<1 0 0>}
      localJointOffset=<1 0 0>
     }]
    relativeSpaceBone {boneName=thigh_l}
   }
   n10 = Anim Dynamics - Bone: dyn_KneeFur_03_l(ComponentPose=@n9, Alpha=1, ExternalForce="0,0,0") {
    angularDampingOverride=0.998
    angularSpringConstant=10000
    bAngularSpring=true
    bOverrideAngularDamping=true
    bUseGravityOverride=true
    alphaBoolBlend {blendOption=Linear}
    boundBone {boneName=dyn_KneeFur_03_l}
    componentAppliedLinearAccClamp=<3000 3000 3000>
    componentLinearAccScale=<0.4 0.4 0.4>
    componentLinearVelScale=<0.3 0.3 0.3>
    gravityOverride {z=-980}
    physicsBodyDefinitions [{
      boundBone {boneName=dyn_KneeFur_03_l}
      boxExtents=<25 25 25>
      constraintSetup {angularLimitsMax=<10 15 20>, angularLimitsMin=<-10 -10 -20>, angularTarget=<1 0 0>}
      localJointOffset=<1 0 0>
     }]
    relativeSpaceBone {boneName=thigh_l}
   }
   n11 = Anim Dynamics - Bone: dyn_KneeFur_01_l(ComponentPose=@n10, Alpha=1, ExternalForce="0,0,0") {
    angularDampingOverride=0.998
    angularSpringConstant=10000
    bAngularSpring=true
    bOverrideAngularDamping=true
    bUseGravityOverride=true
    alphaBoolBlend {blendOption=Linear}
    boundBone {boneName=dyn_KneeFur_01_l}
    componentAppliedLinearAccClamp=<3000 3000 3000>
    componentLinearAccScale=<0.4 0.4 0.4>
    componentLinearVelScale=<0.3 0.3 0.3>
    gravityOverride {z=-980}
    physicsBodyDefinitions [{
      boundBone {boneName=dyn_KneeFur_01_l}
      boxExtents=<25 25 25>
      constraintSetup {angularLimitsMax=<10 15 20>, angularLimitsMin=<-10 -10 -20>, angularTarget=<1 0 0>}
      localJointOffset=<1 0 0>
     }]
    relativeSpaceBone {boneName=thigh_l}
   }
   n12 = Anim Dynamics - Bone: dyn_KneeFur_02_r(ComponentPose=@n11, Alpha=1, ExternalForce="0,0,0") {
    angularDampingOverride=0.998
    angularSpringConstant=10000
    bAngularSpring=true
    bOverrideAngularDamping=true
    bUseGravityOverride=true
    alphaBoolBlend {blendOption=Linear}
    boundBone {boneName=dyn_KneeFur_02_r}
    componentAppliedLinearAccClamp=<3000 3000 3000>
    componentLinearAccScale=<0.4 0.4 0.4>
    componentLinearVelScale=<0.3 0.3 0.3>
    gravityOverride {z=-980}
    physicsBodyDefinitions [{
      boundBone {boneName=dyn_KneeFur_02_r}
      boxExtents=<30 30 30>
      constraintSetup {angularLimitsMax=<10 15 20>, angularLimitsMin=<-10 -10 -20>, angularTarget=<1 0 0>}
      localJointOffset=<2 0 0>
     }]
    relativeSpaceBone {boneName=calf_r}
   }
   n13 = Anim Dynamics - Bone: dyn_KneeFur_03_r(ComponentPose=@n12, Alpha=1, ExternalForce="0,0,0") {
    angularDampingOverride=0.998
    angularSpringConstant=10000
    bAngularSpring=true
    bOverrideAngularDamping=true
    bUseGravityOverride=true
    alphaBoolBlend {blendOption=Linear}
    boundBone {boneName=dyn_KneeFur_03_r}
    componentAppliedLinearAccClamp=<3000 3000 3000>
    componentLinearAccScale=<0.4 0.4 0.4>
    componentLinearVelScale=<0.3 0.3 0.3>
    gravityOverride {z=-980}
    physicsBodyDefinitions [{
      boundBone {boneName=dyn_KneeFur_03_r}
      boxExtents=<30 30 30>
      constraintSetup {angularLimitsMax=<10 15 20>, angularLimitsMin=<-10 -20 -20>, angularTarget=<1 0 0>}
      localJointOffset=<2 0 0>
     }]
    relativeSpaceBone {boneName=calf_r}
   }
   n14 = Anim Dynamics - Bone: dyn_KneeFur_01_r(ComponentPose=@n13, Alpha=1, ExternalForce="0,0,0") {
    angularDampingOverride=0.998
    angularSpringConstant=10000
    bAngularSpring=true
    bOverrideAngularDamping=true
    bUseGravityOverride=true
    alphaBoolBlend {blendOption=Linear}
    boundBone {boneName=dyn_KneeFur_01_r}
    componentAppliedLinearAccClamp=<3000 3000 3000>
    componentLinearAccScale=<0.4 0.4 0.4>
    componentLinearVelScale=<0.3 0.3 0.3>
    gravityOverride {z=-980}
    physicsBodyDefinitions [{
      boundBone {boneName=dyn_KneeFur_01_r}
      boxExtents=<30 30 30>
      constraintSetup {angularLimitsMax=<10 15 30>, angularLimitsMin=<-10 -15 -30>, angularTarget=<1 0 0>}
      localJointOffset=<2 0 0>
     }]
    relativeSpaceBone {boneName=calf_r}
   }
   n15 = Anim Dynamics - Bone: dyn_ElbowFur_01_l(ComponentPose=@n14, Alpha=1, ExternalForce="0,0,0") {
    angularDampingOverride=0.999
    angularSpringConstant=10000
    bAngularSpring=true
    bOverrideAngularDamping=true
    bUseGravityOverride=true
    simulationSpace=BoneRelative
    alphaBoolBlend {blendOption=Linear}
    boundBone {boneName=dyn_ElbowFur_01_l}
    componentAppliedLinearAccClamp=<3000 3000 3000>
    componentLinearAccScale=<0.4 0.4 0.4>
    componentLinearVelScale=<0.3 0.3 0.3>
    gravityOverride {z=-980}
    physicsBodyDefinitions [{
      boundBone {boneName=dyn_ElbowFur_01_l}
      boxExtents=<30 30 30>
      constraintSetup {angularLimitsMax=<10 15 20>, angularLimitsMin=<-10 -10 -20>, angularTarget=<1 0 0>}
      localJointOffset=<1 0 0>
     }]
    relativeSpaceBone {boneName=upperarm_l}
   }
   n16 = Anim Dynamics - Bone: dyn_ElbowFur_02_l(ComponentPose=@n15, Alpha=1, ExternalForce="0,0,0") {
    angularDampingOverride=0.999
    angularSpringConstant=10000
    bAngularSpring=true
    bOverrideAngularDamping=true
    bUseGravityOverride=true
    simulationSpace=BoneRelative
    alphaBoolBlend {blendOption=Linear}
    boundBone {boneName=dyn_ElbowFur_02_l}
    componentAppliedLinearAccClamp=<3000 3000 3000>
    componentLinearAccScale=<0.4 0.4 0.4>
    componentLinearVelScale=<0.3 0.3 0.3>
    gravityOverride {z=-980}
    physicsBodyDefinitions [{
      boundBone {boneName=dyn_ElbowFur_02_l}
      boxExtents=<30 30 30>
      constraintSetup {angularLimitsMax=<10 15 20>, angularLimitsMin=<-10 -10 -20>, angularTarget=<1 0 0>}
      localJointOffset=<1 0 0>
     }]
    relativeSpaceBone {boneName=upperarm_l}
   }
   n17 = Anim Dynamics - Bone: dyn_ElbowFur_03_l(ComponentPose=@n16, Alpha=1, ExternalForce="0,0,0") {
    angularDampingOverride=0.999
    angularSpringConstant=10000
    bAngularSpring=true
    bOverrideAngularDamping=true
    bUseGravityOverride=true
    simulationSpace=BoneRelative
    alphaBoolBlend {blendOption=Linear}
    boundBone {boneName=dyn_ElbowFur_03_l}
    componentAppliedLinearAccClamp=<3000 3000 3000>
    componentLinearAccScale=<0.4 0.4 0.4>
    componentLinearVelScale=<0.3 0.3 0.3>
    gravityOverride {z=-980}
    physicsBodyDefinitions [{
      boundBone {boneName=dyn_ElbowFur_03_l}
      boxExtents=<30 30 30>
      constraintSetup {angularLimitsMax=<10 15 20>, angularLimitsMin=<-10 -10 -20>, angularTarget=<1 0 0>}
      localJointOffset=<1 0 0>
     }]
    relativeSpaceBone {boneName=upperarm_l}
   }
   n18 = Anim Dynamics - Bone: dyn_ElbowFur_01_r(ComponentPose=@n17, Alpha=1, ExternalForce="0,0,0") {
    angularDampingOverride=0.999
    angularSpringConstant=10000
    bAngularSpring=true
    bOverrideAngularDamping=true
    bUseGravityOverride=true
    simulationSpace=BoneRelative
    alphaBoolBlend {blendOption=Linear}
    boundBone {boneName=dyn_ElbowFur_01_r}
    componentAppliedLinearAccClamp=<3000 3000 3000>
    componentLinearAccScale=<0.4 0.4 0.4>
    componentLinearVelScale=<0.3 0.3 0.3>
    gravityOverride {z=-980}
    physicsBodyDefinitions [{
      boundBone {boneName=dyn_ElbowFur_01_r}
      boxExtents=<30 30 30>
      constraintSetup {angularLimitsMax=<10 10 20>, angularLimitsMin=<-10 -15 -20>, angularTarget=<1 0 0>}
      localJointOffset=<1 0 0>
     }]
    relativeSpaceBone {boneName=clavicle_r}
   }
   n19 = Anim Dynamics - Bone: dyn_ElbowFur_02_r(ComponentPose=@n18, Alpha=1, ExternalForce="0,0,0") {
    angularDampingOverride=0.999
    angularSpringConstant=10000
    bAngularSpring=true
    bOverrideAngularDamping=true
    bUseGravityOverride=true
    simulationSpace=BoneRelative
    alphaBoolBlend {blendOption=Linear}
    boundBone {boneName=dyn_ElbowFur_02_r}
    componentAppliedLinearAccClamp=<3000 3000 3000>
    componentLinearAccScale=<0.4 0.4 0.4>
    componentLinearVelScale=<0.3 0.3 0.3>
    gravityOverride {z=-980}
    physicsBodyDefinitions [{
      boundBone {boneName=dyn_ElbowFur_02_r}
      boxExtents=<30 30 30>
      constraintSetup {angularLimitsMax=<10 10 20>, angularLimitsMin=<-10 -15 -20>, angularTarget=<1 0 0>}
      localJointOffset=<1 0 0>
     }]
    relativeSpaceBone {boneName=clavicle_r}
   }
   n20 = Anim Dynamics - Bone: dyn_ElbowFur_03_r(ComponentPose=@n19, Alpha=1, ExternalForce="0,0,0") {
    angularDampingOverride=0.999
    angularSpringConstant=10000
    bAngularSpring=true
    bOverrideAngularDamping=true
    bUseGravityOverride=true
    simulationSpace=BoneRelative
    alphaBoolBlend {blendOption=Linear}
    boundBone {boneName=dyn_ElbowFur_03_r}
    componentAppliedLinearAccClamp=<3000 3000 3000>
    componentLinearAccScale=<0.4 0.4 0.4>
    componentLinearVelScale=<0.3 0.3 0.3>
    gravityOverride {z=-980}
    physicsBodyDefinitions [{
      boundBone {boneName=dyn_ElbowFur_03_r}
      boxExtents=<30 30 30>
      constraintSetup {angularLimitsMax=<10 10 20>, angularLimitsMin=<-10 -15 -20>, angularTarget=<1 0 0>}
      localJointOffset=<1 0 0>
     }]
    relativeSpaceBone {boneName=clavicle_r}
   }
   n21 = Anim Dynamics - Bone: dyn_WristFur_01_l(ComponentPose=@n20, Alpha=1, ExternalForce="0,0,0") {
    angularDampingOverride=0.999
    angularSpringConstant=10000
    bAngularSpring=true
    bOverrideAngularDamping=true
    bUseGravityOverride=true
    simulationSpace=BoneRelative
    alphaBoolBlend {blendOption=Linear}
    boundBone {boneName=dyn_WristFur_01_l}
    componentAppliedLinearAccClamp=<3000 3000 3000>
    componentLinearAccScale=<0.4 0.4 0.4>
    componentLinearVelScale=<0.3 0.3 0.3>
    gravityOverride {z=-980}
    physicsBodyDefinitions [{
      boundBone {boneName=dyn_WristFur_01_l}
      boxExtents=<20 20 20>
      constraintSetup {angularLimitsMax=<10 8 15>, angularLimitsMin=<-10 -25 -15>, angularTarget=<1 0 0>}
      localJointOffset=<1 0 0>
     }]
    relativeSpaceBone {boneName=lowerarm_l}
   }
   n22 = Anim Dynamics - Bone: dyn_WristFur_02_l(ComponentPose=@n21, Alpha=1, ExternalForce="0,0,0") {
    angularDampingOverride=0.999
    angularSpringConstant=10000
    bAngularSpring=true
    bOverrideAngularDamping=true
    bUseGravityOverride=true
    simulationSpace=BoneRelative
    alphaBoolBlend {blendOption=Linear}
    boundBone {boneName=dyn_WristFur_02_l}
    componentAppliedLinearAccClamp=<3000 3000 3000>
    componentLinearAccScale=<0.4 0.4 0.4>
    componentLinearVelScale=<0.3 0.3 0.3>
    gravityOverride {z=-980}
    physicsBodyDefinitions [{
      boundBone {boneName=dyn_WristFur_02_l}
      boxExtents=<20 20 20>
      constraintSetup {angularLimitsMax=<10 8 15>, angularLimitsMin=<-10 -25 -15>, angularTarget=<1 0 0>}
      localJointOffset=<1 0 0>
     }]
    relativeSpaceBone {boneName=lowerarm_l}
   }
   n23 = Anim Dynamics - Bone: dyn_WristFur_01_r(ComponentPose=@n22, Alpha=1, ExternalForce="0,0,0") {
    angularDampingOverride=0.999
    angularSpringConstant=10000
    bAngularSpring=true
    bOverrideAngularDamping=true
    bUseGravityOverride=true
    simulationSpace=BoneRelative
    alphaBoolBlend {blendOption=Linear}
    boundBone {boneName=dyn_WristFur_01_r}
    componentAppliedLinearAccClamp=<3000 3000 3000>
    componentLinearAccScale=<0.4 0.4 0.4>
    componentLinearVelScale=<0.3 0.3 0.3>
    gravityOverride {z=-980}
    physicsBodyDefinitions [{
      boundBone {boneName=dyn_WristFur_01_r}
      boxExtents=<25 25 25>
      constraintSetup {angularLimitsMax=<10 8 15>, angularLimitsMin=<-10 -25 -15>, angularTarget=<1 0 0>}
      localJointOffset=<1 0 0>
     }]
    relativeSpaceBone {boneName=lowerarm_r}
   }
   n24 = Anim Dynamics - Bone: dyn_WristFur_02_r(ComponentPose=@n23, Alpha=1, ExternalForce="0,0,0") {
    angularDampingOverride=0.999
    angularSpringConstant=10000
    bAngularSpring=true
    bOverrideAngularDamping=true
    bUseGravityOverride=true
    simulationSpace=BoneRelative
    alphaBoolBlend {blendOption=Linear}
    boundBone {boneName=dyn_WristFur_02_r}
    componentAppliedLinearAccClamp=<3000 3000 3000>
    componentLinearAccScale=<0.4 0.4 0.4>
    componentLinearVelScale=<0.3 0.3 0.3>
    gravityOverride {z=-980}
    physicsBodyDefinitions [{
      boundBone {boneName=dyn_WristFur_02_r}
      boxExtents=<25 25 25>
      constraintSetup {angularLimitsMax=<10 8 15>, angularLimitsMin=<-10 -25 -15>, angularTarget=<1 0 0>}
      localJointOffset=<1 0 0>
     }]
    relativeSpaceBone {boneName=lowerarm_l}
   }
   Save cached pose 'Dynamics'(Pose=Component To Local(ComponentPose=@n24)) {cacheName=Dynamics}
  ]}
 EventGraph {nodes [
   Event BlueprintUpdateAnimation()
   > n1 = Gravity Override State Function DEPRECATED(CharacterPartAnimInstance=Self-Reference(), GameTimeInSeconds=GetGameTimeInSeconds(), GravityOverrideParams=Get Gravity Override Params())
   > Set GravityOverride(GravityOverride=@n1)
   > Set Is Playing Emote CR(Is Playing Emote CR:double="To Float (Boolean)"(InBool=Get bIsPlayingEmote())) -> (Output_Get:double)
   Event OnEnterGrindRail() {is=title K2Node_Event}
   > Set Grind Rail(Grind Rail:double=1) -> (Output_Get:double)
   Event OnExitGrindRail() {is=title K2Node_Event}
   > Set Grind Rail(Grind Rail=0)
  ]}
]
variables [
 Gravity Override Params {type=GravityOverrideParamsStruct}
 GravityOverride {type=Vector}
 Grind Rail {type=number}
 Is Playing Emote CR {type=number}
]
