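# Corrective Control Rig for the SweetLetter female character's FaceAcc part, posing back and right-side hair base bones as the upper body moves.
# Spherical pose readers on head, neck_01 and right clavicle output weights for chained ModifyTransforms, mostly rotations, on those bones, which parent the dyn_ hair chains.
# On Ground Alpha and is Crouching scale some weights for state-specific corrections.
# Two ModifyTransforms nodes are unconnected, and one head reader has debug drawing enabled.
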ControlRigBlueprint /BRCosmetics/Characters/Player/Female/Medium/Bodies/F_MED_SweetLetter/Meshes/Parts/F_MED_SweetLetter_FaceAcc_CtrlRig
graphs [
 RigVMFunctionLibrary {links [], nodes []}
 RigVMModel {nodes [
   RigUnit_BeginExecution = &RigUnit_BeginExecution()
   n1 = &RigUnit_SphericalPoseReader(
    ActiveRegionScaleFactors=(PositiveWidth=1,NegativeWidth=1,PositiveHeight=0.3,NegativeHeight=0.3)
    ActiveRegionSize=0.15
    Debug=(bDrawDebug=false,bDraw2D=false,bDrawLocalAxes=true,DebugScale=25,DebugSegments=20,DebugThickness=0.25)
    DriverAxis=(X=1,Y=0,Z=0)
    DriverItem=(Type=Bone,Name="head")
    FalloffRegionScaleFactors=(PositiveWidth=1,NegativeWidth=1.3,PositiveHeight=0.5,NegativeHeight=0.5)
    FalloffSize=0.3
    FlipHeightScaling=false
    FlipWidthScaling=true
    OptionalParentItem=(Type=Bone,Name="spine_05")
    RotationOffset=(X=-35,Y=-15,Z=15)
    ExecutePin=@RigUnit_BeginExecution.ExecutePin
   )
   > &RigUnit_ModifyTransforms(
    ItemToModify=((Item=(Type=Bone,Name="base_hairB_R_01"),Transform=(Rotation=(X=-0.2152,Y=0.1882,Z=-0.2152,W=0.9338),Translation=(X=0,Y=0,Z=0),Scale3D=(X=1,Y=1,Z=1))))
    WeightMinimum=0
    Weight=@n1.OutputParam
   )
   > &RigUnit_ModifyTransforms(
    ItemToModify=((Item=(Type=Bone,Name="base_hairB_RR_01"),Transform=(Rotation=(X=0,Y=0,Z=-0.174,W=0.9847),Translation=(X=0,Y=0,Z=0),Scale3D=(X=1,Y=1,Z=1))))
    WeightMinimum=0
    Weight=@n1.OutputParam
   )
   > &RigUnit_ModifyTransforms(
    ItemToModify=((Item=(Type=Bone,Name="base_hairR_01"),Transform=(Rotation=(X=0,Y=-0.0146,Z=0,W=-0.9999),Translation=(X=0,Y=0,Z=0),Scale3D=(X=1,Y=1,Z=1))))
    WeightMinimum=0
    Weight=@n1.OutputParam
   )
   > n2 = &RigUnit_SphericalPoseReader(
    ActiveRegionScaleFactors=(PositiveWidth=0.3,NegativeWidth=0.3,PositiveHeight=0.6,NegativeHeight=1)
    ActiveRegionSize=0.15
    Debug=(bDrawDebug=true,bDraw2D=false,bDrawLocalAxes=true,DebugScale=25,DebugSegments=20,DebugThickness=0.25)
    DriverAxis=(X=1,Y=0,Z=0)
    DriverItem=(Type=Bone,Name="head")
    FalloffRegionScaleFactors=(PositiveWidth=0.8,NegativeWidth=1,PositiveHeight=1.3,NegativeHeight=1)
    FalloffSize=0.4
    FlipHeightScaling=false
    FlipWidthScaling=true
    OptionalParentItem=(Type=Bone,Name="spine_04")
    RotationOffset=(X=0,Y=-62,Z=-90)
   )
   > &RigUnit_ModifyTransforms(
    ItemToModify=((Item=(Type=Bone,Name="base_hairB_C_01"),Transform=(Rotation=(X=0,Y=0,Z=-0.3827,W=0.9239),Translation=(X=0,Y=0,Z=0),Scale3D=(X=1,Y=1,Z=1))))
    WeightMinimum=0
    Weight=@n2.OutputParam
   )
   > &RigUnit_ModifyTransforms(
    ItemToModify=((Item=(Type=Bone,Name="base_hairB_R_01"),Transform=(Rotation=(X=0.2424,Y=0.3462,Z=-0.5198,W=0.7424),Translation=(X=0,Y=0,Z=0),Scale3D=(X=1,Y=1,Z=1))))
    WeightMinimum=0
    Weight=@n2.OutputParam
   )
   > &RigUnit_ModifyTransforms(
    ItemToModify=((Item=(Type=Bone,Name="base_hairB_L_01"),Transform=(Rotation=(X=0.0298,Y=0.0819,Z=-0.3407,W=0.9361),Translation=(X=0,Y=0,Z=0),Scale3D=(X=1,Y=1,Z=1))))
    WeightMinimum=0
    Weight=@n2.OutputParam
   )
   Multiply = &RigVMFunction_MathDoubleMul(A=1, B=RigVMVariableNode() {variable=On Ground Alpha}.Value)
   n3 = &RigVMFunction_MathDoubleMul(B=@Multiply.Result, A=@n2.OutputParam)
   > &RigUnit_ModifyTransforms(
    ItemToModify=((Item=(Type=Bone,Name="base_hairR_01"),Transform=(Rotation=(X=-0.0594,Y=-0.3368,Z=-0.1632,W=0.9254),Translation=(X=0,Y=0,Z=0),Scale3D=(X=1,Y=1,Z=1))))
    WeightMinimum=0
    Weight=@n3.Result
   )
   > &RigUnit_ModifyTransforms(
    ItemToModify=((Item=(Type=Bone,Name="base_hairB_RR_01"),Transform=(Rotation=(X=-0.1294,Y=-0.2241,Z=-0.483,W=0.8365),Translation=(X=0,Y=0,Z=0),Scale3D=(X=1,Y=1,Z=1))))
    WeightMinimum=0
    Weight=@n3.Result
   )
   n4 = &RigVMFunction_MathDoubleMul(B=&RigVMFunction_MathDoubleMul(A=1, B=RigVMVariableNode() {variable=is Crouching}.Value).Result, A=@n2.OutputParam)
   > &RigUnit_ModifyTransforms(
    ItemToModify=((Item=(Type=Bone,Name="base_hairB_C_01"),Transform=(Rotation=(X=-0.0226,Y=-0.2578,Z=-0.0842,W=0.9623),Translation=(X=0,Y=0,Z=0),Scale3D=(X=1,Y=1,Z=1))))
    WeightMinimum=0
    Weight=@n4.Result
   )
   > &RigUnit_ModifyTransforms(
    ItemToModify=((Item=(Type=Bone,Name="base_hairB_L_01"),Transform=(Rotation=(X=0.0151,Y=0.173,Z=-0.0858,W=0.9811),Translation=(X=0,Y=0,Z=0),Scale3D=(X=1,Y=1,Z=1))))
    WeightMinimum=0
    Weight=@n4.Result
   )
   > &RigUnit_ModifyTransforms(
    ItemToModify=((Item=(Type=Bone,Name="base_hairB_R_01"),Transform=(Rotation=(X=-0.0151,Y=-0.173,Z=-0.0858,W=0.9811),Translation=(X=0,Y=0,Z=0),Scale3D=(X=1,Y=1,Z=1))))
    WeightMinimum=0
    Weight=@n4.Result
   )
   > n5 = &RigUnit_SphericalPoseReader(
    ActiveRegionScaleFactors=(PositiveWidth=0.75,NegativeWidth=0.75,PositiveHeight=1,NegativeHeight=1)
    ActiveRegionSize=0.1
    Debug=(bDrawDebug=false,DebugScale=25,DebugSegments=20,DebugThickness=0.25)
    DriverAxis=(X=0,Y=1,Z=0)
    DriverItem=(Type=Bone,Name="head")
    FalloffRegionScaleFactors=(PositiveWidth=1,NegativeWidth=1,PositiveHeight=1,NegativeHeight=2.2)
    FalloffSize=0.2
    OptionalParentItem=(Type=Bone)
    RotationOffset=(X=150,Y=0,Z=0)
   )
   > &RigUnit_ModifyTransforms(
    ItemToModify=((Item=(Type=Bone,Name="base_hairB_L_01"),Transform=(Rotation=(X=-0.0777,Y=-0.1681,Z=-0.0207,W=0.9825),Translation=(X=0,Y=0,Z=0),Scale3D=(X=1,Y=1,Z=1))))
    WeightMinimum=0
    Weight=@n5.OutputParam
   )
   > &RigUnit_ModifyTransforms(
    ItemToModify=((Item=(Type=Bone,Name="base_hairB_C_01"),Transform=(Rotation=(X=-0.2236,Y=-0.3074,Z=-0.0238,W=0.9246),Translation=(X=-0.7642,Y=0.5621,Z=0.919),Scale3D=(X=1,Y=1,Z=1))))
    WeightMinimum=0
    Weight=@n5.OutputParam
   )
   > &RigUnit_ModifyTransforms(
    ItemToModify=((Item=(Type=Bone,Name="base_hairB_R_01"),Transform=(Rotation=(X=-0.253,Y=-0.2628,Z=-0.0503,W=0.9297),Translation=(X=0,Y=0,Z=0),Scale3D=(X=1,Y=1,Z=1))))
    WeightMinimum=0
    Weight=@n5.OutputParam
   )
   > n6 = &RigUnit_SphericalPoseReader(
    ActiveRegionScaleFactors=(PositiveWidth=1,NegativeWidth=1,PositiveHeight=1,NegativeHeight=1)
    ActiveRegionSize=0.15
    Debug=(bDrawDebug=false,bDraw2D=false,bDrawLocalAxes=true,DebugScale=25,DebugSegments=20,DebugThickness=0.25)
    DriverAxis=(X=1,Y=0,Z=0)
    DriverItem=(Type=Bone,Name="neck_01")
    FalloffRegionScaleFactors=(PositiveWidth=1,NegativeWidth=1,PositiveHeight=1,NegativeHeight=1)
    FalloffSize=0.4
    FlipHeightScaling=false
    FlipWidthScaling=true
    OptionalParentItem=(Type=Bone,Name="spine_04")
    RotationOffset=(X=-90,Y=90,Z=150)
   )
   > &RigUnit_ModifyTransforms(
    ItemToModify=((Item=(Type=Bone,Name="base_hairB_L_01"),Transform=(Rotation=(X=0,Y=0,Z=0.2528,W=-0.9675),Translation=(X=0,Y=0,Z=0),Scale3D=(X=1,Y=1,Z=1))))
    WeightMinimum=0
    Weight=@n6.OutputParam
   )
   > n7 = &RigUnit_SphericalPoseReader(
    ActiveRegionScaleFactors=(PositiveWidth=1,NegativeWidth=1,PositiveHeight=1,NegativeHeight=1)
    ActiveRegionSize=0.15
    Debug=(bDrawDebug=false,bDraw2D=false,bDrawLocalAxes=true,DebugScale=25,DebugSegments=20,DebugThickness=0.25)
    DriverAxis=(X=-1,Y=0,Z=0)
    DriverItem=(Type=Bone,Name="clavicle_r")
    FalloffRegionScaleFactors=(PositiveWidth=1,NegativeWidth=1,PositiveHeight=1,NegativeHeight=1)
    FalloffSize=0.4
    FlipHeightScaling=false
    FlipWidthScaling=true
    OptionalParentItem=(Type=Bone,Name="spine_03")
    RotationOffset=(X=110,Y=0,Z=12)
   )
   n8 = RigVMVariableNode() {variable=is Crouching}
   > &RigUnit_ModifyTransforms(
    ItemToModify=((Item=(Type=Bone,Name="base_hairB_R_01"),Transform=(Rotation=(X=0,Y=0.5736,Z=0,W=0.8192),Translation=(X=0,Y=0,Z=0),Scale3D=(X=1,Y=1,Z=1))))
    WeightMinimum=0
    Weight=&RigVMFunction_MathDoubleMul(B=&RigVMFunction_MathDoubleMul(A=1, B=@n8.Value).Result, A=@n7.OutputParam).Result
   )
   &RigUnit_ModifyTransforms(
    ExecutePin=()
    ItemToModify=((Item=(Type=Bone,Name="base_hairB_R_01"),Transform=(Rotation=(X=-0.5042,Y=0.261,Z=-0.5172,W=0.6405),Translation=(X=0,Y=0,Z=0),Scale3D=(X=1,Y=1,Z=1))))
    Weight=1
    WeightMinimum=0
   )
   &RigUnit_ModifyTransforms(
    ExecutePin=()
    ItemToModify=((Item=(Type=Bone,Name="base_hairB_R_01"),Transform=(Rotation=(X=-0.5216,Y=0.2505,Z=-0.1379,W=0.8038),Translation=(X=0,Y=0,Z=0),Scale3D=(X=1,Y=1,Z=1))))
    Weight=1
    WeightMinimum=0
   )
  ]}
]
hierarchy [
 root {loc=<0 0 0>, rot=<0 0 0 1>}
 pelvis {parent=root, loc=<0 -0.5301 98.0571>, rot=<0.0127 0.707 -0.0127 -0.707>}
 spine_01 {parent=pelvis, loc=<6.1208 -0.9328 0>, rot=<0 0 0.034 -0.9994>}
 spine_02 {parent=spine_01, loc=<6.5996 -0.2095 0>, rot=<0 0 0.0238 -0.9997>}
 spine_03 {parent=spine_02, loc=<6.4914 -0.1237 0>, rot=<0 0 0.0149 -0.9999>}
 spine_04 {parent=spine_03, loc=<6.9598 0.0794 0>, rot=<0 0 -0.0132 -0.9999>}
 spine_05 {parent=spine_04, loc=<6.9987 0 0>, rot=<0 0 -0.0015 -1>}
 clavicle_l {parent=spine_05, loc=<5.4831 -0.1053 -3.2946>, rot=<0.0465 -0.7289 -0.009 -0.6829>}
 upperarm_l {parent=clavicle_l, loc=<9.8658 -1.2136 -0.4105>, rot=<0.0164 -0.3245 -0.0269 -0.9454>}
 base_ChestFur_03_l {parent=clavicle_l, loc=<1.7011 0.8288 3.4943>, rot=<0.2601 0.1052 -0.0394 -0.959>}
 dyn_ChestFur_03_l {parent=base_ChestFur_03_l, loc=<0 0 0>, rot=<0 0 0 1>}
 neck_01 {parent=spine_05, loc=<11.3138 0.1529 0>, rot=<0 0 -0.2003 -0.9797>}
 neck_02 {parent=neck_01, loc=<4.5627 0 0>, rot=<0 0 0.0504 -0.9987>}
 head {parent=neck_02, loc=<5.3914 0 0>, rot=<0 0 0.115 -0.9934>}
 base_hairR_01 {parent=head, loc=<5.8617 4.8973 7.4968>, rot=<0.1504 -0.7065 0.6915 -0.0032>}
 dyn_hairR_01 {parent=base_hairR_01, loc=<0 0 0>, rot=<0 0 0 1>}
 dyn_hairR_02 {parent=dyn_hairR_01, loc=<5.8057 -0.0512 -0.147>, rot=<-0.0445 -0.0588 -0.2969 -0.9521>}
 dyn_hairR_03 {parent=dyn_hairR_02, loc=<4.5832 0 -1.0262>, rot=<-0.0153 -0.0662 -0.1703 -0.983>}
 dyn_hairR_04 {parent=dyn_hairR_03, loc=<3.0651 0 0>, rot=<0.0232 -0.0372 0.4574 -0.8882>}
 dyn_hairR_05 {parent=dyn_hairR_04, loc=<3.02 0 0>, rot=<-0.0052 0.0979 0.5615 -0.8216>}
 dyn_hairR_06 {parent=dyn_hairR_05, loc=<2.2947 0 0>, rot=<0.0113 0.0233 0.3187 -0.9475>}
 dyn_hairR_07 {parent=dyn_hairR_06, loc=<2.5212 0 0>, rot=<0.0016 0.0031 0.3526 -0.9358>}
 dyn_hairR_08 {parent=dyn_hairR_02, loc=<5.7352 0.3451 1.5459>, rot=<0.0087 -0.0253 0.1411 -0.9896>}
 dyn_hairR_09 {parent=dyn_hairR_08, loc=<3.4526 0 0>, rot=<0.0015 -0.0201 0.0961 -0.9952>}
 base_hairF_01 {parent=head, loc=<9.3918 11.0788 -4.8902>, rot=<0.3026 -0.7656 -0.5604 -0.091>}
 dyn_hairF_01 {parent=base_hairF_01, loc=<0 0 0>, rot=<0 0 0 1>}
 dyn_hairF_02 {parent=dyn_hairF_01, loc=<4.4512 0 0>, rot=<-0.0551 -0.0264 0.3383 -0.9391>}
 dyn_hairF_03 {parent=dyn_hairF_02, loc=<3.1447 0 0>, rot=<-0.0062 -0.025 0.3239 -0.9457>}
 base_hairB_C_01 {parent=head, loc=<7.1957 -8.1033 0.3833>, rot=<0.061 -0.9981 0 0>}
 dyn_hairB_C_01 {parent=base_hairB_C_01, loc=<0 0 0>, rot=<0 0 0 1>}
 dyn_hairB_C_02 {parent=dyn_hairB_C_01, loc=<8.2212 0 0>, rot=<0 0 0.0828 -0.9966>}
 dyn_hairB_C_03 {parent=dyn_hairB_C_02, loc=<8.2594 0 0>, rot=<0 0 0.0338 -0.9994>}
 dyn_hairB_C_04 {parent=dyn_hairB_C_03, loc=<8.4746 0 0>, rot=<0 0 -0.1226 -0.9925>}
 dyn_hairB_C_05 {parent=dyn_hairB_C_04, loc=<8.1098 0.9016 0>, rot=<-0.0506 0 0 -0.9987>}
 dyn_hairB_C_06 {parent=dyn_hairB_C_03, loc=<6.9707 0 -2.7139>, rot=<-0.073 -0.3015 0.1944 -0.9306>}
 dyn_hairB_C_08 {parent=dyn_hairB_C_06, loc=<4.0295 0 0>, rot=<0.0084 -0.0569 0.1068 -0.9926>}
 base_hairB_R_01 {parent=head, loc=<7.1722 -5.5783 6.1865>, rot=<0.077 -0.9969 0.0104 0.0091>}
 dyn_hairB_R_01 {parent=base_hairB_R_01, loc=<0 0 0>, rot=<0.0096 0.0099 0.0034 -0.9999>}
 dyn_hairB_R_02 {parent=dyn_hairB_R_01, loc=<8.2749 0.6765 -0.0053>, rot=<0 0 0.0734 -0.9973>}
 dyn_hairB_R_03 {parent=dyn_hairB_R_02, loc=<8.1762 -0.6758 -0.64>, rot=<0 0 0.0344 -0.9994>}
 dyn_hairB_R_04 {parent=dyn_hairB_R_03, loc=<7.1579 0 -0.3731>, rot=<0 0 -0.0869 -0.9962>}
 dyn_hairB_R_05 {parent=dyn_hairB_R_04, loc=<5.5926 1.4886 -0.2194>, rot=<0 0 -0.0971 -0.9953>}
 dyn_hairB_R_07 {parent=dyn_hairB_R_01, loc=<4.5851 3.4771 -1.8835>, rot=<-0.5844 -0.0159 -0.0809 -0.8073>}
 dyn_hairB_R_08 {parent=dyn_hairB_R_07, loc=<5.3201 0.1885 0.0101>, rot=<0.0087 0.0096 0.0565 -0.9983>}
 dyn_hairB_R_09 {parent=dyn_hairB_R_08, loc=<4.0143 -0.6868 0.1284>, rot=<0.0185 0.0277 0.5831 -0.8117>}
 base_hairL_01 {parent=head, loc=<4.1783 3.2777 -9.0401>, rot=<-0.0248 -0.7067 -0.7064 0.0306>}
 dyn_hairL_01 {parent=base_hairL_01, loc=<0 0 0>, rot=<0 0 0 1>}
 dyn_hairL_02 {parent=dyn_hairL_01, loc=<7.3467 0.6142 1.3871>, rot=<0.037 -0.031 -0.1652 -0.9851>}
 dyn_hairL_03 {parent=dyn_hairL_02, loc=<4.7645 -0.017 0.0215>, rot=<-0.0523 0.1033 -0.2903 -0.9499>}
 dyn_hairL_04 {parent=dyn_hairL_01, loc=<7.4392 0.0609 -2.2091>, rot=<0.0011 -0.0072 -0.1501 -0.9886>}
 dyn_hairL_05 {parent=dyn_hairL_04, loc=<4.6666 0 0>, rot=<-0.008 0.1386 0.2261 -0.9642>}
 base_hairB_L_01 {parent=head, loc=<7.3192 -5.295 -5.9044>, rot=<0.0956 -0.9954 -0.0005 -0.0047>}
 dyn_hairB_L_01 {parent=base_hairB_L_01, loc=<0 0 0>, rot=<0 -0.0047 0 -1>}
 dyn_hairB_L_02 {parent=dyn_hairB_L_01, loc=<8.2168 0 0>, rot=<0 0 0.0901 -0.9959>}
 dyn_hairB_L_03 {parent=dyn_hairB_L_02, loc=<8.4053 0 0.5102>, rot=<0 0 0.0288 -0.9996>}
 dyn_hairB_L_04 {parent=dyn_hairB_L_03, loc=<7.1171 0 0>, rot=<0 0 -0.1625 -0.9867>}
 dyn_hairB_L_05 {parent=dyn_hairB_L_04, loc=<5.7975 0 0>, rot=<0 0 -0.0474 -0.9989>}
 dyn_hairB_L_07 {parent=dyn_hairB_L_02, loc=<5.3774 2.7542 3.0005>, rot=<0.507 0.3916 0.2126 -0.7378>}
 dyn_hairB_L_06 {parent=dyn_hairB_L_01, loc=<4.3102 0.5786 1.5272>, rot=<0.4134 0.2366 0.2196 -0.8514>}
 base_hairB_RR_01 {parent=head, loc=<6.9091 -2.1649 8.183>, rot=<0.0552 -0.7993 0.5983 0.0105>}
 dyn_hairB_RR_01 {parent=base_hairB_RR_01, loc=<0 0 0>, rot=<0 0 0 1>}
 dyn_hairB_RR_02 {parent=dyn_hairB_RR_01, loc=<4.7639 -0.0018 -0.3315>, rot=<-0.0052 -0.0141 -0.0861 -0.9962>}
 dyn_hairB_RR_03 {parent=dyn_hairB_RR_02, loc=<5.3832 0 -0.2337>, rot=<0.025 0.0883 0.2092 -0.9736>}
 dyn_hairB_RR_04 {parent=dyn_hairB_RR_03, loc=<4.413 0.4678 0.0778>, rot=<0.058 -0.0328 0.4568 -0.8871>}
 clavicle_r {parent=spine_05, loc=<5.4831 -0.1053 3.2946>, rot=<0.7289 0.0465 -0.6829 0.009>}
 upperarm_r {parent=clavicle_r, loc=<-9.8658 1.2136 0.4105>, rot=<0.0164 -0.3245 -0.0269 -0.9454>}
 base_ChestFur_03_r {parent=clavicle_r, loc=<-1.7011 -0.8288 -3.494>, rot=<-0.1053 0.2609 0.9588 -0.0393>}
 dyn_ChestFur_03_r {parent=base_ChestFur_03_r, loc=<0 0 0>, rot=<0 0 0 1>}
 base_ChestFur_01_l {parent=spine_05, loc=<0.1675 10.0643 -0.9942>, rot=<0.2287 -0.6127 -0.2643 -0.7088>}
 dyn_ChestFur_01_l {parent=base_ChestFur_01_l, loc=<0 0 0>, rot=<0 0 0 1>}
 base_ChestFur_02_l {parent=spine_05, loc=<4.7206 5.8656 -4.3112>, rot=<0.214 -0.6299 -0.2431 -0.7059>}
 dyn_ChestFur_02_l {parent=base_ChestFur_02_l, loc=<0 0 0>, rot=<0 0 0 1>}
 base_ChestFur_02_r {parent=spine_05, loc=<4.7201 5.8656 4.3112>, rot=<-0.214 0.6299 -0.2432 -0.7059>}
 dyn_ChestFur_02_r {parent=base_ChestFur_02_r, loc=<0 0 0>, rot=<0 0 0 1>}
 base_ChestFur_01_r {parent=spine_05, loc=<0.167 10.0643 0.9942>, rot=<-0.2282 0.6129 -0.2639 -0.7089>}
 dyn_ChestFur_01_r {parent=base_ChestFur_01_r, loc=<0 0 0>, rot=<0 0 0 1>}
 dyn_backpack_mover {parent=spine_05, loc=<0 0 0>, rot=<0 0 0 1>}
]
variables [
 On Ground Alpha {type=double, default=0}
 is Crouching {type=double, default=0}
]
