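# Physics asset for the front hair face accessory of the female medium SweetLetter player cosmetic.
# Capsule bodies cover front, left and right hair chains, plus the head, neck, upper spine, clavicles, upper arms and chest fur.
# Hair chains start at base bones on the head, with left and right ones branching, and use tight swing and twist limits.
# Base hair, body and chest fur joints leave all angular motion free.
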PhysicsAsset /BRCosmetics/Characters/Player/Female/Medium/Bodies/F_MED_SweetLetter/Meshes/Parts/F_MED_SweetLetter_FaceAcc_FrontHair_Physics
bodies {
 base_hairF_01 [{shapeType=Capsule, length=0.505, radius=0.505, shapeName=base_hairF_01_capsule, rotation=<rot 0 -90 0>}]
 base_hairL_01 [{shapeType=Capsule, length=0.505, radius=0.505, shapeName=base_hairL_01_capsule, rotation=<rot 0 -90 0>}]
 base_hairR_01 [{shapeType=Capsule, length=0.505, radius=0.505, shapeName=base_hairR_01_capsule, rotation=<rot 0 -90 0>}]
 clavicle_l [{shapeType=Capsule, length=10.6, radius=6.8, shapeName=clavicle_l_capsule, center=<8.9405 0.1462 -2.2573>, rotation=<rot 0 -90 90>}]
 clavicle_r [{shapeType=Capsule, length=10.6, radius=6.8, shapeName=clavicle_l_capsule, center=<-6.2526 -0.1462 2.2573>, rotation=<rot 0 -90 90>}]
 dyn_ChestFur_01_l [{shapeType=Capsule, length=5.0234, radius=2, shapeName=dyn_ChestFur_01_l_capsule, center=<3.429 0 0>, rotation=<rot 90 0 0>}]
 dyn_ChestFur_01_r [{shapeType=Capsule, length=5.0234, radius=2, shapeName=dyn_ChestFur_01_r_capsule, center=<3.429 0 0>, rotation=<rot 90 0 0>}]
 dyn_ChestFur_02_l [{shapeType=Capsule, length=5.0234, radius=2, shapeName=dyn_ChestFur_02_l_capsule, center=<4.3173 0 0>, rotation=<rot 90 0 0>}]
 dyn_ChestFur_02_r [{shapeType=Capsule, length=5.0234, radius=2, shapeName=dyn_ChestFur_02_r_capsule, center=<4.3569 0 0>, rotation=<rot 90 0 0>}]
 dyn_ChestFur_03_l [{
   shapeType=Capsule
   length=2.1393
   radius=4.1911
   shapeName=dyn_ChestFur_03_r_capsule
   center=<4.8198 -0.4679 -1.6829>
   rotation=<rot 81.5197 -179.9987 -179.9987>
  }]
 dyn_ChestFur_03_r [{
   shapeType=Capsule
   length=2.1393
   radius=4.1911
   shapeName=dyn_ChestFur_03_r_capsule
   center=<4.8198 0.4679 1.6829>
   rotation=<rot 81.5197 -179.9987 -179.9987>
  }]
 dyn_hairF_01 [{shapeType=Capsule, length=2, radius=1, shapeName=dyn_hairF_01_capsule, center=<2 0 0>, rotation=<rot 90 0 0>}]
 dyn_hairF_02 [{shapeType=Capsule, length=2, radius=1, shapeName=dyn_hairF_02_capsule, center=<2 0 0>, rotation=<rot 90 0 0>}]
 dyn_hairF_03 [{shapeType=Capsule, length=2, radius=1, shapeName=dyn_hairF_03_capsule, center=<2 0 0>, rotation=<rot 90 0 0>}]
 dyn_hairL_01 [{shapeType=Capsule, length=5.3755, radius=1, shapeName=dyn_hairL_01_capsule, center=<3.357 0 0>, rotation=<rot 90 0 0>}]
 dyn_hairL_02 [{shapeType=Capsule, length=2.1883, radius=1.5889, shapeName=dyn_hairL_02_capsule, center=<2 0 -0.1731>, rotation=<rot 90 0 0>}]
 dyn_hairL_03 [{
   shapeType=Capsule
   length=3.3115
   radius=1.1762
   shapeName=dyn_hairL_03_capsule
   center=<1.9075 0 0.3638>
   rotation=<rot 75.7315 179.9999 179.9999>
  }]
 dyn_hairL_04 [{shapeType=Capsule, length=2.3019, radius=1.4992, shapeName=dyn_hairL_04_capsule, center=<1.631 0 0>, rotation=<rot 90 0 0>}]
 dyn_hairL_05 [{
   shapeType=Capsule
   length=2.6621
   radius=1.3706
   shapeName=dyn_hairL_05_capsule
   center=<1.7767 0.003 0.6629>
   rotation=<rot 62.8099 177.0435 -175.3187>
  }]
 dyn_hairR_01 [{shapeType=Capsule, length=2, radius=1, shapeName=dyn_hairR_01_capsule, center=<2 0 0>, rotation=<rot 90 0 0>}]
 dyn_hairR_02 [{shapeType=Capsule, length=2, radius=1, shapeName=dyn_hairR_02_capsule, center=<2 0 0>, rotation=<rot 90 0 0>}]
 dyn_hairR_03 [{shapeType=Capsule, length=2, radius=1, shapeName=dyn_hairR_03_capsule, center=<2 0 0>, rotation=<rot 90 0 0>}]
 dyn_hairR_04 [{shapeType=Capsule, length=2, radius=1, shapeName=dyn_hairR_04_capsule, center=<2 0 0>, rotation=<rot 90 0 0>}]
 dyn_hairR_05 [{shapeType=Capsule, length=2, radius=1, shapeName=dyn_hairR_05_capsule, center=<2 0 0>, rotation=<rot 90 0 0>}]
 dyn_hairR_06 [{shapeType=Capsule, length=2, radius=1, shapeName=dyn_hairR_06_capsule, center=<2 0 0>, rotation=<rot 90 0 0>}]
 dyn_hairR_07 [{shapeType=Capsule, length=1.3009, radius=1, shapeName=dyn_hairR_07_capsule, center=<0.6822 0 0>, rotation=<rot 90 0 0>}]
 dyn_hairR_08 [{shapeType=Capsule, length=1.8573, radius=1.1427, shapeName=dyn_hairR_08_capsule, center=<2 0 0.057>, rotation=<rot 90 0 0>}]
 dyn_hairR_09 [{shapeType=Capsule, length=2.0302, radius=0.9698, shapeName=dyn_hairR_09_capsule, center=<0.9404 0.0949 0>, rotation=<rot 90 0 0>}]
 head [{shapeType=Capsule, length=2.6403, radius=7.3597, center=<2.5678 2.1005 0>, rotation=<rot 90 0 0>}]
 neck_01 [{shapeType=Capsule, length=3.827, radius=5.2754, shapeName=neck_01_capsule, center=<1.2134 0.1918 0>, rotation=<rot 90 0 0>}]
 neck_02 [{shapeType=Capsule, length=0.9198, radius=4.9253, shapeName=neck_02_capsule, center=<1.3913 0.4842 0>, rotation=<rot 0 90 90>}]
 spine_05 [{shapeType=Capsule, length=17.2, radius=8, shapeName=spine_05_capsule, center=<2.8857 -0.0282 0>, rotation=<rot 0 -90 0>}]
 upperarm_l [{shapeType=Capsule, length=13.5146, radius=5.3875, shapeName=upperarm_l_capsule, center=<11.2314 -0.3627 0>, rotation=<rot 90 0 0>}]
 upperarm_r [{shapeType=Capsule, length=13.5146, radius=5.3875, shapeName=upperarm_l_capsule, center=<-11.2314 0.3627 0>, rotation=<rot 90 0 0>}]
}
constraints [
 {bone1Name=dyn_hairF_01, bone2Name=base_hairF_01, swing1LimitDegrees=10, swing2LimitDegrees=10, twistLimitDegrees=1}
 {bone1Name=dyn_hairF_02, bone2Name=dyn_hairF_01, swing1LimitDegrees=10, swing2LimitDegrees=2, twistLimitDegrees=1}
 {bone1Name=dyn_hairF_03, bone2Name=dyn_hairF_02, swing1LimitDegrees=10, swing2LimitDegrees=2, twistLimitDegrees=1}
 {bone1Name=dyn_hairL_01, bone2Name=base_hairL_01, swing1LimitDegrees=20, swing2LimitDegrees=4, twistLimitDegrees=1}
 {bone1Name=dyn_hairL_02, bone2Name=dyn_hairL_01, swing1LimitDegrees=15, swing2LimitDegrees=2, twistLimitDegrees=1}
 {bone1Name=dyn_hairL_03, bone2Name=dyn_hairL_02, swing1LimitDegrees=7, swing2LimitDegrees=5, twistLimitDegrees=1}
 {bone1Name=dyn_hairL_04, bone2Name=dyn_hairL_01, swing1LimitDegrees=12, swing2LimitDegrees=3, twistLimitDegrees=1}
 {bone1Name=dyn_hairL_05, bone2Name=dyn_hairL_04, swing1LimitDegrees=7, swing2LimitDegrees=5, twistLimitDegrees=1}
 {bone1Name=dyn_hairR_01, bone2Name=base_hairR_01, swing1LimitDegrees=15, swing2LimitDegrees=10, twistLimitDegrees=1}
 {bone1Name=dyn_hairR_02, bone2Name=dyn_hairR_01, swing1LimitDegrees=15, swing2LimitDegrees=25, twistLimitDegrees=1}
 {bone1Name=dyn_hairR_03, bone2Name=dyn_hairR_02, swing1LimitDegrees=10, swing2LimitDegrees=7, twistLimitDegrees=1}
 {bone1Name=dyn_hairR_04, bone2Name=dyn_hairR_03, swing1LimitDegrees=15, swing2LimitDegrees=7, twistLimitDegrees=1}
 {bone1Name=dyn_hairR_05, bone2Name=dyn_hairR_04, swing1LimitDegrees=15, swing2LimitDegrees=5, twistLimitDegrees=1}
 {bone1Name=dyn_hairR_06, bone2Name=dyn_hairR_05, swing1LimitDegrees=15, swing2LimitDegrees=5, twistLimitDegrees=1}
 {bone1Name=dyn_hairR_07, bone2Name=dyn_hairR_06, swing1LimitDegrees=15, swing2LimitDegrees=5, twistLimitDegrees=1}
 {bone1Name=dyn_hairR_08, bone2Name=dyn_hairR_02, swing1LimitDegrees=10, swing2LimitDegrees=25, twistLimitDegrees=1}
 {bone1Name=dyn_hairR_09, bone2Name=dyn_hairR_08, swing1LimitDegrees=10, swing2LimitDegrees=20, twistLimitDegrees=1}
 {
  bone1Name=upperarm_l
  bone2Name=clavicle_l
  swing1LimitDegrees=45
  swing1Motion=Free
  swing2LimitDegrees=45
  swing2Motion=Free
  twistLimitDegrees=45
  twistMotion=Free
 }
 {
  bone1Name=upperarm_r
  bone2Name=clavicle_r
  swing1LimitDegrees=45
  swing1Motion=Free
  swing2LimitDegrees=45
  swing2Motion=Free
  twistLimitDegrees=45
  twistMotion=Free
 }
 {
  bone1Name=clavicle_l
  bone2Name=spine_05
  swing1LimitDegrees=45
  swing1Motion=Free
  swing2LimitDegrees=45
  swing2Motion=Free
  twistLimitDegrees=45
  twistMotion=Free
 }
 {
  bone1Name=base_hairR_01
  bone2Name=head
  swing1LimitDegrees=45
  swing1Motion=Free
  swing2LimitDegrees=45
  swing2Motion=Free
  twistLimitDegrees=45
  twistMotion=Free
 }
 {
  bone1Name=base_hairF_01
  bone2Name=head
  swing1LimitDegrees=45
  swing1Motion=Free
  swing2LimitDegrees=45
  swing2Motion=Free
  twistLimitDegrees=45
  twistMotion=Free
 }
 {
  bone1Name=base_hairL_01
  bone2Name=head
  swing1LimitDegrees=45
  swing1Motion=Free
  swing2LimitDegrees=45
  swing2Motion=Free
  twistLimitDegrees=45
  twistMotion=Free
 }
 {
  bone1Name=clavicle_r
  bone2Name=spine_05
  swing1LimitDegrees=45
  swing1Motion=Free
  swing2LimitDegrees=45
  swing2Motion=Free
  twistLimitDegrees=45
  twistMotion=Free
 }
 {
  bone1Name=head
  bone2Name=neck_02
  swing1LimitDegrees=45
  swing1Motion=Free
  swing2LimitDegrees=45
  swing2Motion=Free
  twistLimitDegrees=45
  twistMotion=Free
 }
 {
  bone1Name=neck_01
  bone2Name=spine_05
  swing1LimitDegrees=45
  swing1Motion=Free
  swing2LimitDegrees=45
  swing2Motion=Free
  twistLimitDegrees=45
  twistMotion=Free
 }
 {
  bone1Name=dyn_ChestFur_01_l
  bone2Name=spine_05
  swing1LimitDegrees=45
  swing1Motion=Free
  swing2LimitDegrees=45
  swing2Motion=Free
  twistLimitDegrees=45
  twistMotion=Free
 }
 {
  bone1Name=dyn_ChestFur_01_r
  bone2Name=spine_05
  swing1LimitDegrees=45
  swing1Motion=Free
  swing2LimitDegrees=45
  swing2Motion=Free
  twistLimitDegrees=45
  twistMotion=Free
 }
 {
  bone1Name=dyn_ChestFur_02_l
  bone2Name=spine_05
  swing1LimitDegrees=45
  swing1Motion=Free
  swing2LimitDegrees=45
  swing2Motion=Free
  twistLimitDegrees=45
  twistMotion=Free
 }
 {
  bone1Name=dyn_ChestFur_02_r
  bone2Name=spine_05
  swing1LimitDegrees=45
  swing1Motion=Free
  swing2LimitDegrees=45
  swing2Motion=Free
  twistLimitDegrees=45
  twistMotion=Free
 }
 {
  bone1Name=dyn_ChestFur_03_r
  bone2Name=clavicle_r
  swing1LimitDegrees=45
  swing1Motion=Free
  swing2LimitDegrees=45
  swing2Motion=Free
  twistLimitDegrees=45
  twistMotion=Free
 }
 {
  bone1Name=dyn_ChestFur_03_l
  bone2Name=clavicle_l
  swing1LimitDegrees=45
  swing1Motion=Free
  swing2LimitDegrees=45
  swing2Motion=Free
  twistLimitDegrees=45
  twistMotion=Free
 }
 {
  bone1Name=neck_02
  bone2Name=neck_01
  swing1LimitDegrees=45
  swing1Motion=Free
  swing2LimitDegrees=45
  swing2Motion=Free
  twistLimitDegrees=45
  twistMotion=Free
 }
]
