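# Physics asset for the SweetLetter female medium character's face accessory mesh, setting up secondary motion for its hair.
# Four capsule chains (C, L, R, RR), some branching, hang from head-attached base bones with limited swing and very tight twist.
# Torso, neck, head, upper arm, chest fur and backpack capsules, linked by free constraints, provide apparent collision volumes.
# Free constraints also cross-link the L and R chains to the centre chain at segment 04.
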PhysicsAsset /BRCosmetics/Characters/Player/Female/Medium/Bodies/F_MED_SweetLetter/Meshes/Parts/F_MED_SweetLetter_FaceAcc_Physics
bodies {
 base_hairB_C_01 [{shapeType=Capsule, length=0.505, radius=0.505, shapeName=base_hairB_C_01_capsule, rotation=<rot 0 -90 0>}]
 base_hairB_L_01 [{shapeType=Capsule, length=0.505, radius=0.505, shapeName=base_hairB_L_01_capsule, rotation=<rot 0 -90 0>}]
 base_hairB_RR_01 [{shapeType=Capsule, length=0.505, radius=0.505, shapeName=base_hairB_RR_01_capsule, rotation=<rot 0 -90 0>}]
 base_hairB_R_01 [{shapeType=Capsule, length=0.505, radius=0.505, shapeName=base_hairB_R_01_capsule, rotation=<rot 0 -90 0>}]
 clavicle_l [{shapeType=Capsule, length=10.6, radius=6.8, shapeName=clavicle_l_capsule, center=<6.2526 0.1462 -2.2573>, rotation=<rot 0 -90 90>}]
 clavicle_r [{shapeType=Capsule, length=10.6, radius=6.8, shapeName=clavicle_l_capsule, center=<-6.2526 -0.1462 2.2573>, rotation=<rot 0 -90 90>}]
 dyn_ChestFur_03_l [{
   shapeType=Capsule
   length=2.1393
   radius=4.1911
   shapeName=dyn_ChestFur_03_r_capsule
   center=<4.7046 0.2401 -1.7511>
   rotation=<rot 81.5197 -179.9987 -179.9987>
  }]
 dyn_ChestFur_03_r [
  {
   shapeType=Capsule
   length=2.1393
   radius=4.1911
   shapeName=dyn_ChestFur_03_r_capsule
   center=<4.7571 -0.2401 2.1033>
   rotation=<rot 81.5197 -179.9987 -179.9987>
  }
  {
   shapeType=Capsule
   length=4.8289
   radius=2.7279
   shapeName=dyn_ChestFur_03_r_capsule
   center=<2.8233 -1.4103 1.3477>
   rotation=<rot 71.2228 165.1068 179.9997>
  }
 ]
 dyn_backpack_mover [
  {shapeType=Capsule, length=7.9, radius=19.3, shapeName=dyn_backpack_help_capsule, center=<0 -55.836 0>}
  {shapeType=Capsule, length=7.9, radius=19.3, shapeName=dyn_backpack_help_capsule, center=<10.7681 -56.8748 0>}
 ]
 dyn_hairB_C_01 [{shapeType=Capsule, length=6.0314, radius=1, shapeName=dyn_hairB_C_01_capsule, center=<3.9112 0 0>, rotation=<rot 90 0 0>}]
 dyn_hairB_C_02 [{shapeType=Capsule, length=3.1828, radius=2.5094, shapeName=dyn_hairB_C_02_capsule, center=<3.855 0 0>, rotation=<rot 90 0 0>}]
 dyn_hairB_C_03 [{shapeType=Capsule, length=2.9505, radius=2.5764, shapeName=dyn_hairB_C_03_capsule, center=<3.9492 0 0>, rotation=<rot 90 0 0>}]
 dyn_hairB_C_04 [{shapeType=Capsule, length=3.2663, radius=2.3534, shapeName=dyn_hairB_C_04_capsule, center=<3.7478 0 0>, rotation=<rot 90 0 0>}]
 dyn_hairB_C_05 [{
   shapeType=Capsule
   length=6.3062
   radius=3.1296
   shapeName=dyn_hairB_C_05_capsule
   center=<6.0899 2.0337 0.3157>
   rotation=<rot 81.1591 -161.6468 -180>
  }]
 dyn_hairB_C_06 [{shapeType=Capsule, length=2.2557, radius=1.5668, shapeName=dyn_hairB_C_06_capsule, center=<1.6135 0 0>, rotation=<rot 90 0 0>}]
 dyn_hairB_C_08 [{shapeType=Capsule, length=2.4622, radius=1.0434, shapeName=dyn_hairB_C_08_capsule, center=<2 0 0>, rotation=<rot 90 0 0>}]
 dyn_hairB_L_01 [{shapeType=Capsule, length=6.1739, radius=1, shapeName=dyn_hairB_L_01_capsule, center=<3.7963 0 0>, rotation=<rot 90 0 0>}]
 dyn_hairB_L_02 [{shapeType=Capsule, length=4.5647, radius=2.5402, shapeName=dyn_hairB_L_02_capsule, center=<4.1798 -3.6094 0>, rotation=<rot 90 0 0>}]
 dyn_hairB_L_03 [{
   shapeType=Capsule
   length=2.9023
   radius=3.2924
   shapeName=dyn_hairB_L_03_capsule
   center=<3.4339 -1.3056 1.6341>
   rotation=<rot -0.1388 0.001 0>
  }]
 dyn_hairB_L_04 [{
   shapeType=Capsule
   length=4.0661
   radius=2.9276
   shapeName=dyn_hairB_L_04_capsule
   center=<2.8312 -0.0222 0.3155>
   rotation=<rot -0.6197 0.0001 0>
  }]
 dyn_hairB_L_05 [{
   shapeType=Capsule
   length=3.9593
   radius=2.39
   shapeName=dyn_hairB_L_05_capsule
   center=<4.3995 0.5457 -1.1945>
   rotation=<rot 80.6178 0.0002 -13.1013>
  }]
 dyn_hairB_L_06 [{shapeType=Capsule, length=2, radius=1, shapeName=dyn_hairB_L_06_capsule, center=<2 0 0>, rotation=<rot 90 0 0>}]
 dyn_hairB_L_07 [{
   shapeType=Capsule
   length=2.7241
   radius=1.9364
   shapeName=dyn_hairB_L_07_capsule
   center=<1.0517 -0.0541 0>
   rotation=<rot 90 -165.9638 -135.4086>
  }]
 dyn_hairB_RR_01 [{shapeType=Capsule, length=1, radius=2, shapeName=dyn_hairB_RR_01_capsule, center=<2 0 0>, rotation=<rot 90 0 0>}]
 dyn_hairB_RR_02 [{shapeType=Capsule, length=1.26, radius=1.74, shapeName=dyn_hairB_RR_02_capsule, center=<2 0 0>, rotation=<rot 90 0 0>}]
 dyn_hairB_RR_03 [{shapeType=Capsule, length=2.1204, radius=2.2696, shapeName=dyn_hairB_RR_03_capsule, center=<1.5822 0.1662 -0.2638>, rotation=<rot 90 0 0>}]
 dyn_hairB_RR_04 [{
   shapeType=Capsule
   length=1.905
   radius=1.8048
   shapeName=dyn_hairB_RR_04_capsule
   center=<1.6633 -1.4903 0.6963>
   rotation=<rot 68.0533 164.0584 175.6588>
  }]
 dyn_hairB_R_01 [{shapeType=Capsule, length=6.5288, radius=1, shapeName=dyn_hairB_R_01_capsule, center=<3.6783 0 0>, rotation=<rot 90 0 0>}]
 dyn_hairB_R_02 [{shapeType=Capsule, length=4.0871, radius=3.0007, shapeName=dyn_hairB_R_02_capsule, center=<3.899 -0.3746 0>, rotation=<rot 90 0 0>}]
 dyn_hairB_R_03 [{
   shapeType=Capsule
   length=4.9769
   radius=3.1838
   shapeName=dyn_hairB_R_03_capsule
   center=<2.7319 -0.7954 0>
   rotation=<rot 0.5212 -179.983 179.9996>
  }]
 dyn_hairB_R_04 [{
   shapeType=Capsule
   length=6.28
   radius=2.9343
   shapeName=dyn_hairB_R_04_capsule
   center=<1.0256 -0.013 0.0102>
   rotation=<rot -0.5984 0.0001 0>
  }]
 dyn_hairB_R_05 [{
   shapeType=Capsule
   length=3.8714
   radius=3.0455
   shapeName=dyn_hairB_R_05_capsule
   center=<3.543 -1.6483 0.8025>
   rotation=<rot 77.1867 -179.9998 180>
  }]
 head [{shapeType=Capsule, length=2.6403, radius=7.3597, center=<2.5678 1.1032 0>, rotation=<rot 90 0 0>}]
 neck_01 [{shapeType=Capsule, length=0, radius=7.205, shapeName=neck_01_capsule, center=<1.2134 0.1918 0>, rotation=<rot 90 0 0>}]
 neck_02 [{shapeType=Capsule, length=28.8, radius=15, shapeName=spine_05_capsule, center=<-8.4615 12.9244 0>}]
 spine_02 [{shapeType=Capsule, length=16.7, radius=9.3, shapeName=spine_02_capsule, center=<6.0567 1.2153 0>}]
 spine_03 [{shapeType=Capsule, length=16.7, radius=9.3, shapeName=spine_02_capsule, center=<6.0567 1.1414 0>}]
 spine_04 [{shapeType=Capsule, length=17.8, radius=9.1, shapeName=spine_04_capsule, center=<4.0135 1.6435 0>, rotation=<rot 0 -90 0>}]
 spine_05 [{shapeType=Capsule, length=17.2, radius=8, shapeName=spine_05_capsule, center=<2.8857 0.6609 0>, rotation=<rot 0 -90 0>}]
 upperarm_l [{shapeType=Capsule, length=13.5146, radius=5.3875, shapeName=upperarm_l_capsule, center=<11.2314 -0.3627 0>, rotation=<rot 90 0 0>}]
 upperarm_r [{shapeType=Capsule, length=13.5146, radius=5.3875, shapeName=upperarm_l_capsule, center=<-11.2314 0.3627 0>, rotation=<rot 90 0 0>}]
}
constraints [
 {bone1Name=dyn_hairB_C_01, bone2Name=base_hairB_C_01, swing1LimitDegrees=10, swing2LimitDegrees=15, twistLimitDegrees=1}
 {bone1Name=dyn_hairB_C_02, bone2Name=dyn_hairB_C_01, swing1LimitDegrees=25, swing2LimitDegrees=20, twistLimitDegrees=1}
 {bone1Name=dyn_hairB_C_03, bone2Name=dyn_hairB_C_02, swing1LimitDegrees=25, swing2LimitDegrees=25, twistLimitDegrees=1}
 {bone1Name=dyn_hairB_C_04, bone2Name=dyn_hairB_C_03, swing1LimitDegrees=25, swing2LimitDegrees=5, twistLimitDegrees=1}
 {bone1Name=dyn_hairB_C_05, bone2Name=dyn_hairB_C_04, swing1LimitDegrees=30, swing2LimitDegrees=5, twistLimitDegrees=1}
 {bone1Name=dyn_hairB_C_06, bone2Name=dyn_hairB_C_03, swing1LimitDegrees=10, swing2LimitDegrees=10, twistLimitDegrees=1}
 {bone1Name=dyn_hairB_C_08, bone2Name=dyn_hairB_C_06, swing1LimitDegrees=15, swing2LimitDegrees=15, twistLimitDegrees=1}
 {bone1Name=dyn_hairB_L_01, bone2Name=base_hairB_L_01, swing1LimitDegrees=10, swing2LimitDegrees=15, twistLimitDegrees=1}
 {bone1Name=dyn_hairB_L_02, bone2Name=dyn_hairB_L_01, swing1LimitDegrees=25, swing2LimitDegrees=20, twistLimitDegrees=1}
 {bone1Name=dyn_hairB_L_03, bone2Name=dyn_hairB_L_02, swing1LimitDegrees=15, swing2LimitDegrees=25, twistLimitDegrees=1}
 {bone1Name=dyn_hairB_L_04, bone2Name=dyn_hairB_L_03, swing1LimitDegrees=20, swing2LimitDegrees=6, twistLimitDegrees=1}
 {bone1Name=dyn_hairB_L_05, bone2Name=dyn_hairB_L_04, swing1LimitDegrees=20, swing2LimitDegrees=6, twistLimitDegrees=1}
 {bone1Name=dyn_hairB_L_07, bone2Name=dyn_hairB_L_02, swing1LimitDegrees=10, swing2LimitDegrees=10, twistLimitDegrees=1}
 {bone1Name=dyn_hairB_L_06, bone2Name=dyn_hairB_L_01, swing1LimitDegrees=10, swing2LimitDegrees=10, twistLimitDegrees=1}
 {bone1Name=dyn_hairB_R_01, bone2Name=base_hairB_R_01, swing1LimitDegrees=10, swing2LimitDegrees=15, twistLimitDegrees=1}
 {bone1Name=dyn_hairB_R_02, bone2Name=dyn_hairB_R_01, swing1LimitDegrees=30, swing2LimitDegrees=20, twistLimitDegrees=1}
 {bone1Name=dyn_hairB_R_03, bone2Name=dyn_hairB_R_02, swing1LimitDegrees=30, swing2LimitDegrees=25, twistLimitDegrees=1}
 {bone1Name=dyn_hairB_R_04, bone2Name=dyn_hairB_R_03, swing1LimitDegrees=30, swing2LimitDegrees=5, twistLimitDegrees=1}
 {bone1Name=dyn_hairB_R_05, bone2Name=dyn_hairB_R_04, swing1LimitDegrees=30, swing2LimitDegrees=5, twistLimitDegrees=1}
 {
  bone1Name=upperarm_l
  bone2Name=clavicle_l
  swing1LimitDegrees=45
  swing1Motion=Free
  swing2LimitDegrees=45
  swing2Motion=Free
  twistLimitDegrees=45
  twistMotion=Free
 }
 {
  bone1Name=upperarm_r
  bone2Name=clavicle_r
  swing1LimitDegrees=45
  swing1Motion=Free
  swing2LimitDegrees=45
  swing2Motion=Free
  twistLimitDegrees=45
  twistMotion=Free
 }
 {
  bone1Name=clavicle_l
  bone2Name=spine_05
  swing1LimitDegrees=45
  swing1Motion=Free
  swing2LimitDegrees=45
  swing2Motion=Free
  twistLimitDegrees=45
  twistMotion=Free
 }
 {
  bone1Name=base_hairB_C_01
  bone2Name=head
  swing1LimitDegrees=45
  swing1Motion=Free
  swing2LimitDegrees=45
  swing2Motion=Free
  twistLimitDegrees=45
  twistMotion=Free
 }
 {
  bone1Name=base_hairB_R_01
  bone2Name=head
  swing1LimitDegrees=45
  swing1Motion=Free
  swing2LimitDegrees=45
  swing2Motion=Free
  twistLimitDegrees=45
  twistMotion=Free
 }
 {
  bone1Name=base_hairB_L_01
  bone2Name=head
  swing1LimitDegrees=45
  swing1Motion=Free
  swing2LimitDegrees=45
  swing2Motion=Free
  twistLimitDegrees=45
  twistMotion=Free
 }
 {
  bone1Name=clavicle_r
  bone2Name=spine_05
  swing1LimitDegrees=45
  swing1Motion=Free
  swing2LimitDegrees=45
  swing2Motion=Free
  twistLimitDegrees=45
  twistMotion=Free
 }
 {
  bone1Name=spine_05
  bone2Name=spine_04
  swing1LimitDegrees=45
  swing1Motion=Free
  swing2LimitDegrees=45
  swing2Motion=Free
  twistLimitDegrees=45
  twistMotion=Free
 }
 {
  bone1Name=head
  bone2Name=neck_02
  swing1LimitDegrees=45
  swing1Motion=Free
  swing2LimitDegrees=45
  swing2Motion=Free
  twistLimitDegrees=45
  twistMotion=Free
 }
 {
  bone1Name=neck_01
  bone2Name=spine_05
  swing1LimitDegrees=45
  swing1Motion=Free
  swing2LimitDegrees=45
  swing2Motion=Free
  twistLimitDegrees=45
  twistMotion=Free
 }
 {
  bone1Name=neck_02
  bone2Name=neck_01
  swing1LimitDegrees=45
  swing1Motion=Free
  swing2LimitDegrees=45
  swing2Motion=Free
  twistLimitDegrees=45
  twistMotion=Free
 }
 {
  bone1Name=dyn_hairB_L_04
  bone2Name=dyn_hairB_C_04
  swing1LimitDegrees=45
  swing1Motion=Free
  swing2LimitDegrees=45
  swing2Motion=Free
  twistLimitDegrees=45
  twistMotion=Free
 }
 {
  bone1Name=dyn_hairB_R_04
  bone2Name=dyn_hairB_C_04
  swing1LimitDegrees=45
  swing1Motion=Free
  swing2LimitDegrees=45
  swing2Motion=Free
  twistLimitDegrees=45
  twistMotion=Free
 }
 {
  bone1Name=spine_04
  bone2Name=spine_03
  swing1LimitDegrees=45
  swing1Motion=Free
  swing2LimitDegrees=45
  swing2Motion=Free
  twistLimitDegrees=45
  twistMotion=Free
 }
 {
  bone1Name=spine_03
  bone2Name=spine_02
  swing1LimitDegrees=45
  swing1Motion=Free
  swing2LimitDegrees=45
  swing2Motion=Free
  twistLimitDegrees=45
  twistMotion=Free
 }
 {
  bone1Name=dyn_backpack_mover
  bone2Name=spine_05
  swing1LimitDegrees=45
  swing1Motion=Free
  swing2LimitDegrees=45
  swing2Motion=Free
  twistLimitDegrees=45
  twistMotion=Free
 }
 {
  bone1Name=dyn_ChestFur_03_r
  bone2Name=clavicle_r
  swing1LimitDegrees=45
  swing1Motion=Free
  swing2LimitDegrees=45
  swing2Motion=Free
  twistLimitDegrees=45
  twistMotion=Free
 }
 {
  bone1Name=dyn_ChestFur_03_l
  bone2Name=clavicle_l
  swing1LimitDegrees=45
  swing1Motion=Free
  swing2LimitDegrees=45
  swing2Motion=Free
  twistLimitDegrees=45
  twistMotion=Free
 }
 {
  bone1Name=base_hairB_RR_01
  bone2Name=head
  swing1LimitDegrees=45
  swing1Motion=Free
  swing2LimitDegrees=45
  swing2Motion=Free
  twistLimitDegrees=45
  twistMotion=Free
 }
 {bone1Name=dyn_hairB_RR_01, bone2Name=base_hairB_RR_01, swing1LimitDegrees=10, swing2LimitDegrees=5, twistLimitDegrees=1}
 {bone1Name=dyn_hairB_RR_02, bone2Name=dyn_hairB_RR_01, swing1LimitDegrees=15, swing2LimitDegrees=5, twistLimitDegrees=2}
 {bone1Name=dyn_hairB_RR_03, bone2Name=dyn_hairB_RR_02, swing1LimitDegrees=20, swing2LimitDegrees=10, twistLimitDegrees=3}
 {bone1Name=dyn_hairB_RR_04, bone2Name=dyn_hairB_RR_03, swing1LimitDegrees=25, swing2LimitDegrees=10, twistLimitDegrees=1}
]
