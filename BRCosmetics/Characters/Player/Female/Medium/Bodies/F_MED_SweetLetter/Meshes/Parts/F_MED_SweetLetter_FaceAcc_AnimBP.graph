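# Post-process style AnimBP for the SweetLetter female face accessory, animating hair over the copied body pose.
# Copied poses feed a crouch and skydive-aware Control Rig, then corrective bone offsets.
# Capes or long backblings switch back hair to weighted bone constraints; otherwise backbling state selects between two RigidBody sims, one with an offset backpack-mover bone.
# A final RigidBody always simulates front hair; EventGraph updates gravity override, movement and backbling flags.
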AnimBlueprint /BRCosmetics/Characters/Player/Female/Medium/Bodies/F_MED_SweetLetter/Meshes/Parts/F_MED_SweetLetter_FaceAcc_AnimBP
target_skeleton=&/F_MED_SweetLetter_FaceAcc_Skeleton
graphs [
 AnimGraph {nodes [
   n1 = "Transform (Modify) Bone - Bone: dyn_backpack_mover"!(ComponentPose=Local To Component(LocalPose=Use cached pose 'Correctives'()), bAlphaBoolEnabled:bool=True, Translation="0,22,0") -> (Pose) {
    alphaInputType=Bool
    translationMode=BMM_Additive
    translationSpace=BCS_BoneSpace
    alphaBoolBlend {blendOption=Linear}
    boneToModify {boneName=dyn_backpack_mover}
    rotation {roll=20}
    translation {y=22}
   }
   n2 = RigidBody!(
    ComponentPose=@n1
    bAlphaBoolEnabled:bool
    OverrideWorldGravity:Vector="0,0,-980"
    ComponentLinearAccScale:Vector="0.01,0.01,0.01"
    ComponentLinearVelScale:Vector="0.01,0.01,0.01"
   ) -> (Pose) {
    alphaInputType=Bool
    bClampLinearTranslationLimitToRefPose=true
    bOverrideWorldGravity=true
    overridePhysicsAsset=&/F_MED_SweetLetter_FaceAcc_Physics
    simulationSpace=BaseBoneSpace
    alphaBoolBlend {blendOption=Linear}
    baseBoneRef {boneName=spine_05}
    componentAppliedLinearAccClamp=<1000 1000 1000>
    componentLinearAccScale=<0.6 0.2 0.2>
    componentLinearVelScale=<0.6 0.2 0.2>
   }
   n3 = RigidBody(
    ComponentPose=Local To Component(LocalPose=Use cached pose 'Correctives'())
    bAlphaBoolEnabled=Get isWearingBackbling_not() -> (isWearingBackbling_not:bool)
    OverrideWorldGravity=Get GravityOverride()
    ComponentLinearAccScale="0.13,0.24,0.11"
    ComponentLinearVelScale="0.13,0.24,0.11"
   ) {
    alphaInputType=Bool
    bClampLinearTranslationLimitToRefPose=true
    bOverrideWorldGravity=true
    overridePhysicsAsset=&F_MED_SweetLetter_FaceAcc_Physics
    simulationSpace=BaseBoneSpace
    alphaBoolBlend {blendOption=Linear}
    baseBoneRef {boneName=neck_01}
    componentAppliedLinearAccClamp=<3000 3000 3000>
    componentLinearAccScale=<0.6 0.2 0.2>
    componentLinearVelScale=<0.6 0.2 0.2>
    simSpaceSettings {worldAlpha=0.2}
   }
   n4 = Blend Poses by bool(BlendPose_0=Component To Local(ComponentPose=@n2), BlendPose_1=Component To Local(ComponentPose=@n3), bActiveValue=Get isWearingBackbling()) {blendType=Linear}
   n5 = "Constraint - dyn_hairB_C_01 (P)"(ComponentPose=Local To Component(LocalPose=Use cached pose 'Input Pose'()), Alpha=1, ConstraintWeights_0=1) {
    alphaBoolBlend {blendOption=Linear}
    boneToModify {boneName=dyn_hairB_C_01}
    constraintSetup [{targetBone {boneName=head}}]
    constraintWeights [1]
   }
   n6 = "Constraint - dyn_hairB_C_02 (PP)"!(ComponentPose=@n5, Alpha=1, ConstraintWeights_0=0.8, ConstraintWeights_1:float=0.2) -> (Pose) {
    alphaBoolBlend {blendOption=Linear}
    boneToModify {boneName=dyn_hairB_C_02}
    constraintSetup [{targetBone {boneName=head}}, {targetBone {boneName=neck_02}}]
    constraintWeights [1,1]
   }
   n7 = "Constraint - dyn_hairB_C_03 (PPP)"!(ComponentPose=@n6, Alpha=1, ConstraintWeights_0=0.2, ConstraintWeights_1:float=0, ConstraintWeights_2:float=0.8) -> (Pose) {
    alphaBoolBlend {blendOption=Linear}
    boneToModify {boneName=dyn_hairB_C_03}
    constraintSetup [{targetBone {boneName=neck_01}}, {targetBone {boneName=neck_02}}, {targetBone {boneName=spine_05}}]
    constraintWeights [1,1,1]
   }
   n8 = "Constraint - dyn_hairB_C_04 (PPPP)"!(
    ComponentPose=@n7
    Alpha=1
    ConstraintWeights_0=0.6447
    ConstraintWeights_1:float=0.2157
    ConstraintWeights_2:float=0.1245
    ConstraintWeights_3:float=0.015
   ) -> (Pose) {
    alphaBoolBlend {blendOption=Linear}
    boneToModify {boneName=dyn_hairB_C_04}
    constraintSetup [{targetBone {boneName=spine_05}}, {targetBone {boneName=clavicle_r}}, {targetBone {boneName=clavicle_l}}, {targetBone {boneName=spine_04}}]
    constraintWeights [1,1,1,1]
   }
   n9 = "Constraint - dyn_hairB_C_05 (PPPPP)"!(
    ComponentPose=@n8
    Alpha=1
    ConstraintWeights_0=0.7
    ConstraintWeights_1:float=0.3
    ConstraintWeights_2:float=0
    ConstraintWeights_3:float=0
    ConstraintWeights_4:float=0
   ) -> (Pose) {
    alphaBoolBlend {blendOption=Linear}
    boneToModify {boneName=dyn_hairB_C_05}
    constraintSetup [
     {targetBone {boneName=spine_05}}
     {targetBone {boneName=spine_04}}
     {targetBone {boneName=spine_03}}
     {targetBone {boneName=spine_02}}
     {targetBone {boneName=spine_01}}
    ]
    constraintWeights [1,1,1,1,1]
   }
   n10 = "Constraint - dyn_hairB_L_01 (P)"(ComponentPose=@n9, Alpha=1, ConstraintWeights_0=1) {
    alphaBoolBlend {blendOption=Linear}
    boneToModify {boneName=dyn_hairB_L_01}
    constraintSetup [{targetBone {boneName=head}}]
    constraintWeights [1]
   }
   n11 = "Constraint - dyn_hairB_L_02 (PP)"!(ComponentPose=@n10, Alpha=1, ConstraintWeights_0=0.8, ConstraintWeights_1:float=0.2) -> (Pose) {
    alphaBoolBlend {blendOption=Linear}
    boneToModify {boneName=dyn_hairB_L_02}
    constraintSetup [{targetBone {boneName=head}}, {targetBone {boneName=neck_02}}]
    constraintWeights [1,1]
   }
   n12 = "Constraint - dyn_hairB_L_03 (PPPP)"!(
    ComponentPose=@n11
    Alpha=1
    ConstraintWeights_0=0.5103
    ConstraintWeights_1:float=0.4846
    ConstraintWeights_2:float=0.0027
    ConstraintWeights_3:float=0.0024
   ) -> (Pose) {
    alphaBoolBlend {blendOption=Linear}
    boneToModify {boneName=dyn_hairB_L_03}
    constraintSetup [{targetBone {boneName=clavicle_l}}, {targetBone {boneName=spine_05}}, {targetBone {boneName=neck_01}}, {targetBone {boneName=spine_03}}]
    constraintWeights [1,1,1,1]
   }
   n13 = "Constraint - dyn_hairB_L_04 (PPPP)"!(
    ComponentPose=@n12
    Alpha=1
    ConstraintWeights_0=0
    ConstraintWeights_1:float=0.7
    ConstraintWeights_2:float=0.3
    ConstraintWeights_3:float=0
   ) -> (Pose) {
    alphaBoolBlend {blendOption=Linear}
    boneToModify {boneName=dyn_hairB_L_04}
    constraintSetup [{targetBone {boneName=clavicle_l}}, {targetBone {boneName=spine_05}}, {targetBone {boneName=spine_04}}, {targetBone {boneName=clavicle_r}}]
    constraintWeights [1,1,1,1]
   }
   n14 = "Constraint - dyn_hairB_L_05 (PPPP)"!(
    ComponentPose=@n13
    Alpha=1
    ConstraintWeights_0=0
    ConstraintWeights_1:float=0.8
    ConstraintWeights_2:float=0.2
    ConstraintWeights_3:float=0
   ) -> (Pose) {
    alphaBoolBlend {blendOption=Linear}
    boneToModify {boneName=dyn_hairB_L_05}
    constraintSetup [{targetBone {boneName=clavicle_l}}, {targetBone {boneName=spine_05}}, {targetBone {boneName=spine_04}}, {targetBone {boneName=clavicle_r}}]
    constraintWeights [1,1,1,1]
   }
   n15 = "Constraint - dyn_hairB_R_01 (P)"(ComponentPose=@n14, Alpha=1, ConstraintWeights_0=1) {
    alphaBoolBlend {blendOption=Linear}
    boneToModify {boneName=dyn_hairB_R_01}
    constraintSetup [{targetBone {boneName=head}}]
    constraintWeights [1]
   }
   n16 = "Constraint - dyn_hairB_R_02 (PP)"!(ComponentPose=@n15, Alpha=1, ConstraintWeights_0=0.8, ConstraintWeights_1:float=0.2) -> (Pose) {
    alphaBoolBlend {blendOption=Linear}
    boneToModify {boneName=dyn_hairB_R_02}
    constraintSetup [{targetBone {boneName=head}}, {targetBone {boneName=neck_02}}]
    constraintWeights [1,1]
   }
   n17 = "Constraint - dyn_hairB_R_03 (PPPP)"!(
    ComponentPose=@n16
    Alpha=1
    ConstraintWeights_0=0.5108
    ConstraintWeights_1:float=0.4848
    ConstraintWeights_2:float=0.0023
    ConstraintWeights_3:float=0.0021
   ) -> (Pose) {
    alphaBoolBlend {blendOption=Linear}
    boneToModify {boneName=dyn_hairB_R_03}
    constraintSetup [{targetBone {boneName=clavicle_r}}, {targetBone {boneName=spine_05}}, {targetBone {boneName=neck_01}}, {targetBone {boneName=spine_03}}]
    constraintWeights [1,1,1,1]
   }
   n18 = "Constraint - dyn_hairB_R_04 (PPP)"!(ComponentPose=@n17, Alpha=1, ConstraintWeights_0=0, ConstraintWeights_1:float=1, ConstraintWeights_2:float=0) -> (Pose) {
    alphaBoolBlend {blendOption=Linear}
    boneToModify {boneName=dyn_hairB_R_04}
    constraintSetup [{targetBone {boneName=clavicle_r}}, {targetBone {boneName=spine_05}}, {targetBone {boneName=spine_04}}]
    constraintWeights [1,1,1]
   }
   n19 = "Constraint - dyn_hairB_R_05 (PP)"!(ComponentPose=@n18, Alpha=1, ConstraintWeights_0=1, ConstraintWeights_1:float=0) -> (Pose) {
    alphaBoolBlend {blendOption=Linear}
    boneToModify {boneName=dyn_hairB_R_05}
    constraintSetup [{targetBone {boneName=spine_05}}, {targetBone {boneName=spine_04}}]
    constraintWeights [1,1]
   }
   n20 = "Constraint - dyn_hairB_R_07 (P)"(ComponentPose=@n19, Alpha=1, ConstraintWeights_0=1) {
    alphaBoolBlend {blendOption=Linear}
    boneToModify {boneName=dyn_hairB_R_07}
    constraintSetup [{targetBone {boneName=head}}]
    constraintWeights [1]
   }
   n21 = Component To Local(ComponentPose=@n20)
   n22 = Get Is Wearing Cape() -> (Is Wearing Cape:bool)
   n23 = RigidBody(
    ComponentPose=Local To Component(LocalPose=Blend Poses by bool(BlendPose_0=@n21, BlendPose_1=@n4, bActiveValue=@n22) {blendType=Linear})
    OverrideWorldGravity="0,0,0"
    ComponentLinearAccScale="0.12,0.12,0.12"
    ComponentLinearVelScale="0.12,0.12,0.12"
   ) {
    alphaInputType=Bool
    bClampLinearTranslationLimitToRefPose=true
    bOverrideWorldGravity=true
    overridePhysicsAsset=&/F_MED_SweetLetter_FaceAcc_FrontHair_Physics
    simulationSpace=BaseBoneSpace
    alphaBoolBlend {blendOption=Linear}
    baseBoneRef {boneName=spine_05}
    componentAppliedLinearAccClamp=<3000 3000 5000>
    componentLinearAccScale=<0.6 0.2 0.2>
    componentLinearVelScale=<0.6 0.2 0.2>
    simSpaceSettings {worldAlpha=0.2}
   }
   Output Pose(Result=Component To Local(ComponentPose=@n23)) {layerGroup=None}
   n24 = Layered blend per bone(
    BasePose=Copy Pose From Mesh() {bUseAttachedParent=true}
    BlendPoses_0=Copy Pose From Mesh(SourceMeshComponent=Get Body_Mesh()) {bUseAttachedParent=true}
   ) {layerSetup [{branchFilters [{boneName=clavicle_l}, {boneName=clavicle_r}, {boneName=upperarm_r}, {boneName=upperarm_l}]}]}
   n25 = Control Rig!(
    Source=@n24
    Alpha
    On Ground Alpha:double="To Float (Boolean)"(InBool=Get On Ground CR() -> (On Ground CR:bool))
    is Crouching:double="To Float (Boolean)"(InBool=Get is Crouching() -> (is Crouching:bool))
   ) -> (Pose) {
    alphaBoolBlend {blendOption=Linear}
    controlRigAssetReference {blueprintRigClass=&/F_MED_SweetLetter_FaceAcc_CtrlRig.F_MED_SweetLetter_FaceAcc_CtrlRig_C}
   }
   Save cached pose 'Input Pose'(Pose=@n25) {cacheName=Input Pose}
   n26 = "Transform (Modify) Bone - Bone: base_hairB_L_01"(
    ComponentPose=Local To Component(LocalPose=Use cached pose 'Input Pose'())
    Alpha=1
    Translation="0,0,0"
    Rotation="-0.012,-27.064,0"
    Scale="1,1,1"
   ) {
    rotationMode=BMM_Additive
    rotationSpace=BCS_BoneSpace
    translationMode=BMM_Additive
    translationSpace=BCS_BoneSpace
    alphaBoolBlend {blendOption=Linear}
    alphaScaleBiasClamp {bInterpResult=true}
    boneToModify {boneName=base_hairB_L_01}
    rotation {pitch=-0.012, yaw=-27.064}
   }
   n27 = "Transform (Modify) Bone - Bone: base_hairB_C_01"(ComponentPose=@n26, Alpha=1, Translation="0,0,0", Rotation="0,0,0", Scale="1,1,1") {
    rotationMode=BMM_Additive
    rotationSpace=BCS_BoneSpace
    translationMode=BMM_Additive
    translationSpace=BCS_BoneSpace
    alphaBoolBlend {blendOption=Linear}
    alphaScaleBiasClamp {bInterpResult=true}
    boneToModify {boneName=base_hairB_C_01}
   }
   Save cached pose 'Correctives'(Pose=Component To Local(ComponentPose=@n27)) {cacheName=Correctives}
  ]}
 EventGraph {nodes [
   Event OnCharacterCustomizationLoaded()
   > n1 = Cast To PlayerPawn_Athena(Object=TryGetPawnOwner())
   > Set Body_Mesh(Body_Mesh=GetSkeletalMeshForPartType(self=@n1))
   Event BlueprintInitializeAnimation()
   > Set Skel Mesh(Skel Mesh=GetOwnerSkeletalMeshComponent())
   Event BlueprintUpdateAnimation()
   > n2 = Sequence!(execute) -> (then_0, then_1, then_2:Exec)
   Set Hair Up(
    execute=@n2.then_1
    Hair Up:double=GetAlphaFromReferenceTransformDelta(SkeletalMeshComponent=Get Skel Mesh(), BoneName=spine_05, BaseName=head, InputRange=(X=0,Y=70), TwistAxis="0,0,1")
   ) -> (Output_Get:double)
   > n3 = PlayerMovement(AnimInstance=Self-Reference(), XDotProductCoefficient=100, YDotProductCoefficient=100, ZDotProductCoefficient=100)
   n4 = BreakVector(InVec=MakeVector(X=@n3, Y=@n3.YMovement, Z=@n3.ZMovement))
   > n5 = Set Left(Left:bool="float < float"(A=@n4, B=-50)) -> (Output_Get:bool)
   > n6 = Set Right(Right:bool="float > float"(A=@n4, B=50)) -> (Output_Get:bool)
   > Set IsLeft(IsLeft:double="To Float (Boolean)"(InBool=@n5)) -> (Output_Get:double)
   > Set IsRight(IsRight:double="To Float (Boolean)"(InBool=@n6)) -> (Output_Get:double)
   n7 = Gravity Override State Function DEPRECATED(
    execute=@n2.then_0
    CharacterPartAnimInstance=Self-Reference()
    GameTimeInSeconds=GetGameTimeInSeconds()
    GravityOverrideParams=Get Gravity Override Params()
    DebugJointName=spine_05
   )
   > Set GravityOverride(GravityOverride=@n7)
   n8 = CheckForBackbling()
   n9 = Set isWearingBackbling(execute=@n2.then_2, isWearingBackbling=@n8)
   > Set isWearingBackbling_not(isWearingBackbling_not:bool=NOT Boolean(A=@n9)) -> (Output_Get:bool)
   > Set Is Wearing Cape(Is Wearing Cape=OR Boolean(A=@n8.bIsBackblingACape, B=@n8.bIsBackblingLong))
   > Set On Ground CR(On Ground CR:bool=NOT Boolean(A=Get bIsSkyDiving())) -> (Output_Get:bool)
   > Set OnGround CR(OnGround CR:double="To Float (Boolean)"(InBool=Get bIsOnGround() -> (bIsOnGround:bool))) -> (Output_Get:double)
   > Set is Crouching(is Crouching:bool=Get bIsCrouching()) -> (Output_Get:bool)
  ]}
]
variables [
 Body_Mesh {type=SkeletalMeshComponent}
 Gravity Override Params {type=GravityOverrideParamsStruct}
 GravityOverride {type=Vector}
 Hair Up {type=number}
 is Crouching {type=boolean}
 Is Wearing Cape {type=boolean}
 IsLeft {type=number}
 IsRight {type=number}
 isWearingBackbling {type=boolean}
 isWearingBackbling_not {type=boolean}
 Left {type=boolean}
 On Ground CR {type=boolean}
 OnGround CR {type=number}
 Right {type=boolean}
 Skel Mesh {type=SkeletalMeshComponent}
]
